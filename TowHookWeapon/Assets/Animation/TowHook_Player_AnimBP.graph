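# Player Animation Blueprint for the TowHook weapon's tethered animations, built in the SecondMovementMode_FinalPoseOverride layer.
# Lean-driven reposition states pass through foot-target and leg IK, take additive noise, then blend with jump and yank state machines.
# A mesh-space Spine_04 layered blend restores the incoming upper-body pose.
# Mounting, additive lean, a pelvis-to-attach bone copy and an Attach state machine follow, with inertialized blending on output.
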AnimBlueprint /TowHookWeapon/Assets/Animation/TowHook_Player_AnimBP
target_skeleton=&Fortnite_M_Avg_Player_Skeleton
graphs [
 AnimGraph {nodes [Output Pose()]}
 EventGraph {links [], nodes []}
 MovementMode_AimOffset {nodes [
   Input Pose() {name=InAimOffsetsPose}
   Input Pose() {name=InSourcePosePreAimOffsets}
   Output Pose() {layerGroup=MovementMode}
  ]}
 MovementMode_BaseLocomotion {nodes [
   Input Pose() {name=InPoseBaseLocomotion}
   Output Pose() {layerGroup=MovementMode}
  ]}
 MovementMode_FinalPoseOverride {nodes [
   Input Pose() {name=InFinalPose}
   Output Pose() {layerGroup=MovementMode}
  ]}
 MovementMode_FullBodyOverride {nodes [
   Input Pose() {name=InputPoseFullBody}
   Output Pose() {layerGroup=MovementMode}
  ]}
 MovementMode_InAirOverride {nodes [
   Input Pose() {name=InSource_Pose}
   Input Pose() {name=InAirPose}
   Output Pose() {layerGroup=MovementMode}
  ]}
 MovementMode_LocomotionAdditives {nodes [
   Input Pose() {name=InLocomotionAdditivesPose}
   Output Pose() {layerGroup=MovementMode}
  ]}
 MovementMode_LowerBodyOverride {nodes [
   Input Pose() {name=InLowerBodyPose}
   Input Pose() {name=InSourcePose}
   Output Pose() {layerGroup=MovementMode}
  ]}
 MovementMode_OffScreen {nodes [
   Input Pose() {name=InPoseOffscreen}
   Output Pose() {layerGroup=MovementMode}
  ]}
 MovementMode_SourcePose {nodes [
   Input Pose() {name=InSourcePoseBase}
   Output Pose() {layerGroup=MovementMode}
  ]}
 SecondMovementMode_FinalPoseOverride {nodes [
   n1 = Apply Additive(
    Base=Use cached pose 'PreAddLean Pose'()
    Additive=Blendspace Player 'BS_TowHook_Tethered_Reposition_Additive'(X=1, Y=0) {blendSpace=&/BS_TowHook_Tethered_Reposition_Additive}
   ) {alphaBoolBlend {blendOption=Linear}}
   n2 = Use cached pose 'PreAddLean Pose'()
   n3 = Blend Poses by bool(
    BlendPose_0=Blend Poses by bool(
     BlendPose_0=Layered blend per bone(BasePose=@n1, BlendPoses_0=@n2) {bMeshSpaceRotationBlend=true, layerSetup [{branchFilters [{blendDepth=3, boneName=Spine_04}]}]}
     BlendPose_1=Use cached pose 'PreAddLean Pose'()
     BlendTime_0=0.2
     BlendTime_1=0.3
    ) {blendType=Linear}
    BlendPose_1=Use cached pose 'FinalPose Cached'()
    BlendTime_1=0.2
   ) {blendType=Linear}
   n4 = "Transform (Modify) Bone - Bone: attach"(
    ComponentPose=Copy Bone - Source Bone: pelvis - Target Bone: attach!(
     ComponentPose:Component Space Pose Link Structure=Local To Component(LocalPose=@n3)
     Alpha:float=1
     bCopyTranslation:bool=true
     bCopyRotation:bool=true
     bCopyScale:bool=False
    ) -> (Pose:Component Space Pose Link Structure) {is=title AnimGraphNode_CopyBone, alphaBoolBlend {blendOption=Linear}, sourceBone {boneName=pelvis}, targetBone {boneName=attach}}
    Alpha=1
    Translation="0,0,5"
    Rotation="0,0,0"
    Scale="1,1,1"
   ) {translationMode=BMM_Additive, alphaBoolBlend {blendOption=Linear}, boneToModify {boneName=attach}, translation {z=5}}
   n5 = Component To Local(ComponentPose=@n4)
   AnimGraphNode_StateMachine_3 = Attach SM() {is=title AnimGraphNode_StateMachine}
   Output Pose(Result=Inertialization!(Source:Pose Link Structure=Blend Poses by bool(BlendPose_0=@AnimGraphNode_StateMachine_3, BlendPose_1=@n5, BlendTime_0=0, BlendTime_1=0.5) {blendType=Linear, transitionType=Inertialization}) -> (Pose:Pose Link Structure) {is=title AnimGraphNode_Inertialization}) {layerGroup=SecondMovementMode}
   Save cached pose 'FinalPose Cached'(Pose=Input Pose() {name=InFinalPose}) {cacheName=FinalPose Cached, cachePoseName=FinalPose Cached}
   n6 = Blend Poses by int!(
    BlendPose_0:Pose Link Structure="Sequence Player (additive) 'TowHook_Tethered_Noise'"!(PlayRate:float=1) -> (Pose) {sequence=&/TowHook_Tethered_Noise}
    BlendPose_1:Pose Link Structure="Sequence Player (additive) 'TowHook_Tethered_Noise_Water'"!(PlayRate:float=1) -> (Pose) {sequence=&/TowHook_Tethered_Noise_Water}
    BlendTime_0:float=0.1
    BlendTime_1:float=0.1
    ActiveChildIndex:int=Get NoisePicker() -> (NoisePicker:int=0)
   ) -> (Pose:Pose Link Structure) {
    is=title AnimGraphNode_BlendListByInt
    blendType=Linear
    blendPose [{linkId=-1, sourceLinkId=-1}, {linkId=-1, sourceLinkId=-1}]
    blendTime [0.1,0.1]
   }
   n7 = Blend Poses by int!(
    BlendPose_0="Sequence Player (additive) 'TowHook_Tethered_Base_Pose_Additive'"() {bLoopAnimation=false, sequence=&/TowHook_Tethered_Base_Pose_Additive}
    BlendPose_1="Sequence Player (additive) 'TowHook_Tethered_Base_Alt_Pose_v1'"() {bLoopAnimation=false, sequence=&/TowHook_Tethered_Base_Alt_Pose_v1}
    BlendPose_2:Pose Link Structure="Sequence Player (additive) 'TowHook_Tethered_Base_Alt_Pose_v2'"() {bLoopAnimation=false, sequence=&/TowHook_Tethered_Base_Alt_Pose_v2}
    BlendPose_3:Pose Link Structure="Sequence Player (additive) 'TowHook_Tethered_Base_Alt_Pose_v3'"() {bLoopAnimation=false, sequence=&/TowHook_Tethered_Base_Alt_Pose_v3}
    BlendTime_0=0.2
    BlendTime_1=0.2
    BlendTime_2:float=0.3
    BlendTime_3:float=0.2
    ActiveChildIndex=Get PoseShift() -> (PoseShift:int=0)
   ) -> (Pose) {
    blendType=Linear
    blendPose [{linkId=-1, sourceLinkId=-1}, {linkId=-1, sourceLinkId=-1}, {linkId=-1, sourceLinkId=-1}, {linkId=-1, sourceLinkId=-1}]
    blendTime [0.1,0.1,0.1,0.1]
   }
   n8 = Apply Additive(
    Base=Two Way Blend(
     A=Sequence Player 'TowHook_Tethered_Base_Pose'() {bLoopAnimation=false, sequence=&/TowHook_Tethered_Base_Pose}
     B=Blendspace Player 'BS_TowHook_Tethered_Reaction_Leans'(X=0, Y=0) {blendSpace=&/BS_TowHook_Tethered_Reaction_Leans}
     Alpha=0
    ) {blendNode {alphaBoolBlend {blendOption=Linear}}}
    Additive=@n7
   ) {alphaBoolBlend {blendOption=Linear}}
   AnimGraphNode_StateMachine_1 = Tethered Shift SM() {is=title AnimGraphNode_StateMachine}
   n9 = Set Foot IK Targets!(
    ComponentPose:Component Space Pose Link Structure=Local To Component(LocalPose=Blend Poses by bool(BlendPose_0=@AnimGraphNode_StateMachine_1, BlendPose_1=@n8, BlendTime_0=0.2, BlendTime_1=0.2) {blendType=Linear})
    Alpha:float=1
   ) -> (Pose:Component Space Pose Link Structure) {
    is=title FortAnimGraphNode_SetIKTarget
    iKTraceLength=70
    maxFootHeightComponentSpace=33
    alphaBoolBlend {blendOption=Linear}
    footL {boneName=foot_l}
    footR {boneName=foot_r}
    iKFootL {boneName=ik_foot_l}
    iKFootR {boneName=ik_foot_r}
   }
   n10 = Leg IK!(ComponentPose:Component Space Pose Link Structure=@n9, Alpha:float=1, SoftPercentLength:float=1) -> (Pose:Component Space Pose Link Structure) {
    is=title AnimGraphNode_LegIK
    softPercentLength=0.95
    alphaBoolBlend {blendOption=Linear}
    legsDefinition [
     {
      bEnableKneeTwistCorrection=false
      bEnableRotationLimit=false
      footBoneForwardAxis=Y
      hingeRotationAxis=Z
      minRotationAngle=15
      numBonesInLimb=2
      twistOffsetCurveName=None
      fKFootBone {boneName=foot_l}
      iKFootBone {boneName=ik_foot_l}
     }
     {
      bEnableKneeTwistCorrection=false
      bEnableRotationLimit=false
      footBoneForwardAxis=Y
      hingeRotationAxis=Z
      minRotationAngle=15
      numBonesInLimb=2
      twistOffsetCurveName=None
      fKFootBone {boneName=foot_r}
      iKFootBone {boneName=ik_foot_r}
     }
    ]
   }
   n11 = Component To Local(ComponentPose=@n10)
   AnimGraphNode_StateMachine_0 = Tethered Jump SM() {is=title AnimGraphNode_StateMachine, bAllowConduitEntryStates=true}
   Save cached pose 'Pre Yank Pose'(Pose=Blend Poses by bool(
     BlendPose_0=@AnimGraphNode_StateMachine_0
     BlendPose_1=Apply Additive(Base=@n11, Additive=@n6) {alphaBoolBlend {blendOption=Linear}}
     BlendTime_0=0.2
     BlendTime_1=0.4
    ) {blendType=Linear}) {cacheName=Pre Yank Pose}
   AnimGraphNode_StateMachine_2 = Yank SM() {is=title AnimGraphNode_StateMachine, bAllowConduitEntryStates=true}
   n12 = Layered blend per bone(
    BasePose=Blend Poses by bool(BlendPose_0=@AnimGraphNode_StateMachine_2, BlendPose_1=Use cached pose 'Pre Yank Pose'(), BlendTime_0=0.2, BlendTime_1=0.3) {blendType=Linear}
    BlendPoses_0=Use cached pose 'FinalPose Cached'()
   ) {bMeshSpaceRotationBlend=true, layerSetup [{branchFilters [{blendDepth=3, boneName=Spine_04}]}]}
   n13 = Use cached pose 'FinalPose Cached'()
   AnimGraphNode_StateMachine_4 = Mounting SM() {is=title AnimGraphNode_StateMachine, bAllowConduitEntryStates=true}
   Save cached pose 'PreAddLean Pose'(Pose=Blend Poses by bool(
     BlendPose_0=@AnimGraphNode_StateMachine_4
     BlendPose_1=Blend Poses by bool(BlendPose_0=@n13, BlendPose_1=@n12, BlendTime_1=0.2) {blendType=Linear}
     BlendTime_1=0
    ) {blendType=Linear}) {cacheName=PreAddLean Pose}
  ]}
 Tethered Jump SM {
  entry=Jump or Fall?
  path=SecondMovementMode_FinalPoseOverride.AnimGraphNode_StateMachine_0.Tethered Jump SM
  ids=id
  nodes [
   ToFalling() {is=title AnimStateAliasNode, also {aliases [Land]}}
   ToJumping() {is=title AnimStateAliasNode, also {aliases [Land]}}
   AnimStateConduitNode_0 = Jump or Fall?() {is=title AnimStateConduitNode}
   AnimStateNode_0 = Jumping() {is=title AnimStateNode}
   AnimStateNode_1 = Falling() {is=title AnimStateNode}
   AnimStateNode_2 = PreLand() {is=title AnimStateNode}
   AnimStateNode_3 = Land() {is=title AnimStateNode}
   AnimStateTransitionNode_0 = AnimStateTransitionNode()
   AnimStateTransitionNode_1 = AnimStateTransitionNode() {priorityOrder=10}
   AnimStateTransitionNode_2 = AnimStateTransitionNode() {bAutomaticRuleBasedOnSequencePlayerInState=true}
   AnimStateTransitionNode_3 = AnimStateTransitionNode() {crossfadeDuration=0.1}
   AnimStateTransitionNode_4 = AnimStateTransitionNode() {crossfadeDuration=0.1}
   AnimStateTransitionNode_5 = AnimStateTransitionNode()
   AnimStateTransitionNode_6 = AnimStateTransitionNode() {crossfadeDuration=0.15}
  ]
  links [
   {
    dst=PreLand
    rule=SecondMovementMode_FinalPoseOverride.AnimGraphNode_StateMachine_0.Tethered Jump SM.AnimStateTransitionNode_3.Transition
    src=Falling
    via=AnimStateTransitionNode_3
   }
   {
    dst=Falling
    rule=SecondMovementMode_FinalPoseOverride.AnimGraphNode_StateMachine_0.Tethered Jump SM.AnimStateTransitionNode_1.Transition
    src=Jump or Fall?
    via=AnimStateTransitionNode_1
   }
   {
    dst=Jumping
    rule=SecondMovementMode_FinalPoseOverride.AnimGraphNode_StateMachine_0.Tethered Jump SM.AnimStateTransitionNode_0.Transition
    src=Jump or Fall?
    via=AnimStateTransitionNode_0
   }
   {
    dst=Falling
    rule=SecondMovementMode_FinalPoseOverride.AnimGraphNode_StateMachine_0.Tethered Jump SM.AnimStateTransitionNode_2.Transition
    src=Jumping
    via=AnimStateTransitionNode_2
   }
   {
    dst=Land
    rule=SecondMovementMode_FinalPoseOverride.AnimGraphNode_StateMachine_0.Tethered Jump SM.AnimStateTransitionNode_4.Transition
    src=PreLand
    via=AnimStateTransitionNode_4
   }
   {
    dst=Falling
    rule=SecondMovementMode_FinalPoseOverride.AnimGraphNode_StateMachine_0.Tethered Jump SM.AnimStateTransitionNode_5.Transition
    src=ToFalling
    via=AnimStateTransitionNode_5
   }
   {
    dst=Jumping
    rule=SecondMovementMode_FinalPoseOverride.AnimGraphNode_StateMachine_0.Tethered Jump SM.AnimStateTransitionNode_6.Transition
    src=ToJumping
    via=AnimStateTransitionNode_6
   }
  ]
 }
 Jump or Fall? {
  path=SecondMovementMode_FinalPoseOverride.AnimGraphNode_StateMachine_0.Tethered Jump SM.AnimStateConduitNode_0.Jump or Fall?
  nodes [Result(bCanEnterTransition=true)]
 }
 Jumping {
  path=SecondMovementMode_FinalPoseOverride.AnimGraphNode_StateMachine_0.Tethered Jump SM.AnimStateNode_0.Jumping
  nodes [
   n1 = Blend Poses by int!(
    BlendPose_0=Sequence Player 'TowHook_Tethered_Jump_C_Start'() {bLoopAnimation=false, sequence=&/TowHook_Tethered_Jump_C_Start}
    BlendPose_1=Sequence Player 'TowHook_Tethered_Jump_L_Start'() {bLoopAnimation=false, sequence=&/TowHook_Tethered_Jump_L_Start}
    BlendPose_2=Sequence Player 'TowHook_Tethered_Jump_R_Start'() {bLoopAnimation=false, sequence=&/TowHook_Tethered_Jump_R_Start}
    BlendTime_0=0.1
    BlendTime_1=0.1
    BlendTime_2=0.1
    ActiveChildIndex=Get JumpPicker() -> (JumpPicker:int=0)
   ) -> (Pose) {
    blendType=Linear
    blendPose [{linkId=-1, sourceLinkId=-1}, {linkId=-1, sourceLinkId=-1}, {linkId=-1, sourceLinkId=-1}]
    blendTime [0.1,0.1,0.1]
   }
   Output Animation Pose(Result=Apply Additive(
     Base=@n1
     Additive="Sequence Player (additive) 'TowHook_Tethered_Noise_Air'"!(PlayRate:float=1) -> (Pose) {bLoopAnimation=false, sequence=&/TowHook_Tethered_Noise_Air}
    ) {alphaBoolBlend {blendOption=Linear}})
  ]
 }
 Falling {
  path=SecondMovementMode_FinalPoseOverride.AnimGraphNode_StateMachine_0.Tethered Jump SM.AnimStateNode_1.Falling
  nodes [
   n1 = Blend Poses by int(
    BlendPose_0=Sequence Player 'TowHook_Tethered_Jump_C_Loop'() {bLoopAnimation=false, sequence=&/TowHook_Tethered_Jump_C_Loop}
    BlendPose_1=Sequence Player 'TowHook_Tethered_Jump_L_Loop'() {bLoopAnimation=false, sequence=&/TowHook_Tethered_Jump_L_Loop}
    BlendPose_2=Sequence Player 'TowHook_Tethered_Jump_R_Loop'() {bLoopAnimation=false, sequence=&/TowHook_Tethered_Jump_R_Loop}
    BlendTime_0=0.1
    BlendTime_1=0.1
    BlendTime_2=0.1
    ActiveChildIndex=Get JumpPicker() -> (JumpPicker=0)
   ) {
    blendType=Linear
    blendPose [{linkId=-1, sourceLinkId=-1}, {linkId=-1, sourceLinkId=-1}, {linkId=-1, sourceLinkId=-1}]
    blendTime [0.1,0.1,0.1]
   }
   Output Animation Pose(Result=Apply Additive(
     Base=@n1
     Additive="Sequence Player (additive) 'TowHook_Tethered_Noise_Air'"(PlayRate=1) {bLoopAnimation=false, sequence=&TowHook_Tethered_Noise_Air}
    ) {alphaBoolBlend {blendOption=Linear}})
  ]
 }
 PreLand {
  path=SecondMovementMode_FinalPoseOverride.AnimGraphNode_StateMachine_0.Tethered Jump SM.AnimStateNode_2.PreLand
  nodes [
   n1 = Blend Poses by int(
    BlendPose_0=Sequence Player 'TowHook_Tethered_Jump_C_PreLand'() {bLoopAnimation=false, sequence=&/TowHook_Tethered_Jump_C_PreLand}
    BlendPose_1=Sequence Player 'TowHook_Tethered_Jump_L_PreLand'() {bLoopAnimation=false, sequence=&/TowHook_Tethered_Jump_L_PreLand}
    BlendPose_2=Sequence Player 'TowHook_Tethered_Jump_R_PreLand'() {bLoopAnimation=false, sequence=&/TowHook_Tethered_Jump_R_PreLand}
    BlendTime_0=0.1
    BlendTime_1=0.1
    BlendTime_2=0.1
    ActiveChildIndex=Get JumpPicker() -> (JumpPicker=0)
   ) {
    blendType=Linear
    blendPose [{linkId=-1, sourceLinkId=-1}, {linkId=-1, sourceLinkId=-1}, {linkId=-1, sourceLinkId=-1}]
    blendTime [0.1,0.1,0.1]
   }
   Output Animation Pose(Result=Apply Additive(
     Base=@n1
     Additive="Sequence Player (additive) 'TowHook_Tethered_Noise_Air'"(PlayRate=1) {bLoopAnimation=false, sequence=&TowHook_Tethered_Noise_Air}
    ) {alphaBoolBlend {blendOption=Linear}})
  ]
 }
 Land {
  path=SecondMovementMode_FinalPoseOverride.AnimGraphNode_StateMachine_0.Tethered Jump SM.AnimStateNode_3.Land
  nodes [Output Animation Pose(Result=Sequence Player 'TowHook_Tethered_Jump_Land'() {bLoopAnimation=false, sequence=&/TowHook_Tethered_Jump_Land, startPosition=0.2})]
 }
 Transition {
  path=SecondMovementMode_FinalPoseOverride.AnimGraphNode_StateMachine_0.Tethered Jump SM.AnimStateTransitionNode_0.Transition
  nodes [Result(bCanEnterTransition=Get bIsJumping())]
 }
 Transition {
  path=SecondMovementMode_FinalPoseOverride.AnimGraphNode_StateMachine_0.Tethered Jump SM.AnimStateTransitionNode_1.Transition
  nodes [Result(bCanEnterTransition=Get bIsFalling())]
 }
 Transition {
  path=SecondMovementMode_FinalPoseOverride.AnimGraphNode_StateMachine_0.Tethered Jump SM.AnimStateTransitionNode_2.Transition
  nodes [Result()]
 }
 Transition {
  path=SecondMovementMode_FinalPoseOverride.AnimGraphNode_StateMachine_0.Tethered Jump SM.AnimStateTransitionNode_3.Transition
  nodes [
   Result(bCanEnterTransition=Get bLandingPredicted() -> (bLandingPredicted:bool))
   NOT Boolean(A=Get bIsTetheredJumping() -> (bIsTetheredJumping:bool))
  ]
 }
 Transition {
  path=SecondMovementMode_FinalPoseOverride.AnimGraphNode_StateMachine_0.Tethered Jump SM.AnimStateTransitionNode_4.Transition
  nodes [
   Result(bCanEnterTransition=NOT Boolean(A=Get bIsFalling()))
   NOT Boolean(A=Get bIsTetheredJumping())
  ]
 }
 Transition {
  path=SecondMovementMode_FinalPoseOverride.AnimGraphNode_StateMachine_0.Tethered Jump SM.AnimStateTransitionNode_5.Transition
  nodes [Result(bCanEnterTransition=Get bIsFalling())]
 }
 Transition {
  path=SecondMovementMode_FinalPoseOverride.AnimGraphNode_StateMachine_0.Tethered Jump SM.AnimStateTransitionNode_6.Transition
  nodes [Result(bCanEnterTransition=Get bIsJumping())]
 }
 Tethered Shift SM {
  entry=Idle
  path=SecondMovementMode_FinalPoseOverride.AnimGraphNode_StateMachine_1.Tethered Shift SM
  ids=id
  nodes [
   To Idle() {is=title AnimStateAliasNode, also {aliases [Backward End, Forward End, Leftt End, Right End]}}
   To Left to Right() {is=title AnimStateAliasNode, also {aliases [Left Loop, Leftt End, Right to Left]}}
   To Right to Left() {is=title AnimStateAliasNode, also {aliases [Left to Right, Right End, Right Loop]}}
   To Forward Start() {
    is=title AnimStateAliasNode
    also {aliases [Idle, Left Loop, Left Start, Left to Right, Leftt End, Right End, Right Loop, Right Start, Right to Left]}
   }
   To Backward Start() {
    is=title AnimStateAliasNode
    also {aliases [Idle, Left Loop, Left Start, Left to Right, Leftt End, Right End, Right Loop, Right Start, Right to Left]}
   }
   FrontOrBackToLeft() {is=title AnimStateAliasNode, also {aliases [Backward End, Backward Loop, Forward End, Forward Loop]}}
   FrontOrBackToRight() {is=title AnimStateAliasNode, also {aliases [Backward End, Backward Loop, Forward End, Forward Loop]}}
   AnimStateNode_34 = Idle() {
    stateEntered {guid=#1}
    stateFullyBlended {guid=#2}
    stateLeft {guid=#3}
   }
   AnimStateNode_35 = Right Loop() {
    is=title AnimStateNode
    stateEntered {guid=#4}
    stateFullyBlended {guid=#5}
    stateLeft {guid=#6}
   }
   AnimStateNode_36 = Left Loop() {
    is=title AnimStateNode
    stateEntered {guid=#7}
    stateFullyBlended {guid=#8}
    stateLeft {guid=#9}
   }
   AnimStateNode_37 = Left to Right() {
    is=title AnimStateNode
    stateEntered {guid=#10}
    stateFullyBlended {guid=#11}
    stateLeft {guid=#12}
   }
   AnimStateNode_38 = Right to Left() {is=title AnimStateNode}
   AnimStateNode_39 = Forward Loop() {is=title AnimStateNode}
   AnimStateNode_40 = Backward Loop() {is=title AnimStateNode}
   AnimStateNode_41 = Backward Start() {
    is=title AnimStateNode
    stateEntered {guid=#13}
    stateFullyBlended {guid=#14}
    stateLeft {guid=#15}
   }
   AnimStateNode_42 = Backward End() {
    is=title AnimStateNode
    stateEntered {guid=#16}
    stateFullyBlended {guid=#17}
    stateLeft {guid=#18}
   }
   AnimStateNode_43 = Forward End() {
    is=title AnimStateNode
    stateEntered {guid=#19}
    stateFullyBlended {guid=#20}
    stateLeft {guid=#21}
   }
   AnimStateNode_44 = Forward Start() {
    is=title AnimStateNode
    stateEntered {guid=#22}
    stateFullyBlended {guid=#23}
    stateLeft {guid=#24}
   }
   AnimStateNode_45 = Right Start() {
    is=title AnimStateNode
    stateEntered {guid=#22}
    stateFullyBlended {guid=#23}
    stateLeft {guid=#24}
   }
   AnimStateNode_46 = Right End() {
    is=title AnimStateNode
    stateEntered {guid=#19}
    stateFullyBlended {guid=#20}
    stateLeft {guid=#21}
   }
   AnimStateNode_47 = Leftt End() {
    is=title AnimStateNode
    stateEntered {guid=#19}
    stateFullyBlended {guid=#20}
    stateLeft {guid=#21}
   }
   AnimStateNode_48 = Left Start() {
    is=title AnimStateNode
    stateEntered {guid=#22}
    stateFullyBlended {guid=#23}
    stateLeft {guid=#24}
   }
   AnimStateNode_49 = BacktoRight() {is=title AnimStateNode}
   AnimStateNode_50 = BacktoLeft() {is=title AnimStateNode}
   AnimStateTransitionNode_0 = AnimStateTransitionNode() {crossfadeDuration=0.1}
   AnimStateTransitionNode_1 = AnimStateTransitionNode() {crossfadeDuration=0.1}
   AnimStateTransitionNode_2 = AnimStateTransitionNode()
   AnimStateTransitionNode_64 = AnimStateTransitionNode()
   AnimStateTransitionNode_65 = AnimStateTransitionNode() {bAutomaticRuleBasedOnSequencePlayerInState=true, crossfadeDuration=0.1}
   AnimStateTransitionNode_66 = AnimStateTransitionNode()
   AnimStateTransitionNode_67 = AnimStateTransitionNode() {bAutomaticRuleBasedOnSequencePlayerInState=true, crossfadeDuration=0.1}
   AnimStateTransitionNode_68 = AnimStateTransitionNode() {crossfadeDuration=0.1}
   AnimStateTransitionNode_71 = AnimStateTransitionNode() {bAutomaticRuleBasedOnSequencePlayerInState=true, crossfadeDuration=0.1}
   AnimStateTransitionNode_72 = AnimStateTransitionNode()
   AnimStateTransitionNode_74 = AnimStateTransitionNode() {crossfadeDuration=0.1}
   AnimStateTransitionNode_76 = AnimStateTransitionNode() {crossfadeDuration=0.1}
   AnimStateTransitionNode_77 = AnimStateTransitionNode() {bAutomaticRuleBasedOnSequencePlayerInState=true, crossfadeDuration=0.1}
   AnimStateTransitionNode_78 = AnimStateTransitionNode() {bAutomaticRuleBasedOnSequencePlayerInState=true}
   AnimStateTransitionNode_79 = AnimStateTransitionNode() {bAutomaticRuleBasedOnSequencePlayerInState=true}
   AnimStateTransitionNode_87 = AnimStateTransitionNode() {bAutomaticRuleBasedOnSequencePlayerInState=true}
   AnimStateTransitionNode_89 = AnimStateTransitionNode() {bAutomaticRuleBasedOnSequencePlayerInState=true}
   AnimStateTransitionNode_90 = AnimStateTransitionNode() {bAutomaticRuleBasedOnSequencePlayerInState=true, crossfadeDuration=0.1}
   AnimStateTransitionNode_91 = AnimStateTransitionNode()
   AnimStateTransitionNode_92 = AnimStateTransitionNode()
   AnimStateTransitionNode_93 = AnimStateTransitionNode() {crossfadeDuration=0.1}
  ]
  links [
   {
    dst=Left Loop
    rule=SecondMovementMode_FinalPoseOverride.AnimGraphNode_StateMachine_1.Tethered Shift SM.AnimStateTransitionNode_89.AnimationTransitionGraph_0
    src=BacktoLeft
    via=AnimStateTransitionNode_89
   }
   {
    dst=Right Loop
    rule=SecondMovementMode_FinalPoseOverride.AnimGraphNode_StateMachine_1.Tethered Shift SM.AnimStateTransitionNode_87.AnimationTransitionGraph_0
    src=BacktoRight
    via=AnimStateTransitionNode_87
   }
   {
    dst=Backward End
    rule=SecondMovementMode_FinalPoseOverride.AnimGraphNode_StateMachine_1.Tethered Shift SM.AnimStateTransitionNode_64.AnimationTransitionGraph_0
    src=Backward Loop
    via=AnimStateTransitionNode_64
   }
   {
    dst=Backward Loop
    rule=SecondMovementMode_FinalPoseOverride.AnimGraphNode_StateMachine_1.Tethered Shift SM.AnimStateTransitionNode_65.AnimationTransitionGraph_0
    src=Backward Start
    via=AnimStateTransitionNode_65
   }
   {
    dst=Forward End
    rule=SecondMovementMode_FinalPoseOverride.AnimGraphNode_StateMachine_1.Tethered Shift SM.AnimStateTransitionNode_72.AnimationTransitionGraph_0
    src=Forward Loop
    via=AnimStateTransitionNode_72
   }
   {
    dst=Forward Loop
    rule=SecondMovementMode_FinalPoseOverride.AnimGraphNode_StateMachine_1.Tethered Shift SM.AnimStateTransitionNode_71.AnimationTransitionGraph_0
    src=Forward Start
    via=AnimStateTransitionNode_71
   }
   {
    dst=BacktoLeft
    rule=SecondMovementMode_FinalPoseOverride.AnimGraphNode_StateMachine_1.Tethered Shift SM.AnimStateTransitionNode_2.Transition
    src=FrontOrBackToLeft
    via=AnimStateTransitionNode_2
   }
   {
    dst=BacktoRight
    rule=SecondMovementMode_FinalPoseOverride.AnimGraphNode_StateMachine_1.Tethered Shift SM.AnimStateTransitionNode_1.Transition
    src=FrontOrBackToRight
    via=AnimStateTransitionNode_1
   }
   {
    dst=Left Start
    rule=SecondMovementMode_FinalPoseOverride.AnimGraphNode_StateMachine_1.Tethered Shift SM.AnimStateTransitionNode_76.AnimationTransitionGraph_0
    src=Idle
    via=AnimStateTransitionNode_76
   }
   {
    dst=Right Start
    rule=SecondMovementMode_FinalPoseOverride.AnimGraphNode_StateMachine_1.Tethered Shift SM.AnimStateTransitionNode_66.AnimationTransitionGraph_0
    src=Idle
    via=AnimStateTransitionNode_66
   }
   {
    dst=Leftt End
    rule=SecondMovementMode_FinalPoseOverride.AnimGraphNode_StateMachine_1.Tethered Shift SM.AnimStateTransitionNode_74.AnimationTransitionGraph_0
    src=Left Loop
    via=AnimStateTransitionNode_74
   }
   {
    dst=Left Loop
    rule=SecondMovementMode_FinalPoseOverride.AnimGraphNode_StateMachine_1.Tethered Shift SM.AnimStateTransitionNode_77.AnimationTransitionGraph_0
    src=Left Start
    via=AnimStateTransitionNode_77
   }
   {
    dst=Right Loop
    rule=SecondMovementMode_FinalPoseOverride.AnimGraphNode_StateMachine_1.Tethered Shift SM.AnimStateTransitionNode_79.AnimationTransitionGraph_0
    src=Left to Right
    via=AnimStateTransitionNode_79
   }
   {
    dst=Right End
    rule=SecondMovementMode_FinalPoseOverride.AnimGraphNode_StateMachine_1.Tethered Shift SM.AnimStateTransitionNode_68.AnimationTransitionGraph_0
    src=Right Loop
    via=AnimStateTransitionNode_68
   }
   {
    dst=Right Loop
    rule=SecondMovementMode_FinalPoseOverride.AnimGraphNode_StateMachine_1.Tethered Shift SM.AnimStateTransitionNode_67.AnimationTransitionGraph_0
    src=Right Start
    via=AnimStateTransitionNode_67
   }
   {
    dst=Left Loop
    rule=SecondMovementMode_FinalPoseOverride.AnimGraphNode_StateMachine_1.Tethered Shift SM.AnimStateTransitionNode_78.AnimationTransitionGraph_0
    src=Right to Left
    via=AnimStateTransitionNode_78
   }
   {
    dst=Backward Start
    rule=SecondMovementMode_FinalPoseOverride.AnimGraphNode_StateMachine_1.Tethered Shift SM.AnimStateTransitionNode_93.Transition
    src=To Backward Start
    via=AnimStateTransitionNode_93
   }
   {
    dst=Forward Start
    rule=SecondMovementMode_FinalPoseOverride.AnimGraphNode_StateMachine_1.Tethered Shift SM.AnimStateTransitionNode_0.Transition
    src=To Forward Start
    via=AnimStateTransitionNode_0
   }
   {
    dst=Idle
    rule=SecondMovementMode_FinalPoseOverride.AnimGraphNode_StateMachine_1.Tethered Shift SM.AnimStateTransitionNode_90.Transition
    src=To Idle
    via=AnimStateTransitionNode_90
   }
   {
    dst=Left to Right
    rule=SecondMovementMode_FinalPoseOverride.AnimGraphNode_StateMachine_1.Tethered Shift SM.AnimStateTransitionNode_91.Transition
    src=To Left to Right
    via=AnimStateTransitionNode_91
   }
   {
    dst=Right to Left
    rule=SecondMovementMode_FinalPoseOverride.AnimGraphNode_StateMachine_1.Tethered Shift SM.AnimStateTransitionNode_92.Transition
    src=To Right to Left
    via=AnimStateTransitionNode_92
   }
  ]
 }
 Idle {
  path=SecondMovementMode_FinalPoseOverride.AnimGraphNode_StateMachine_1.Tethered Shift SM.AnimStateNode_34.Idle
  nodes [
   Modify Curve(CurveValues_0=1, Alpha=1) {curveNames [UserCurve01], curveValues [0]}
   Output Animation Pose(Result=Sequence Player 'TowHook_Tethered_Base_Pose'() {bLoopAnimation=false, sequence=&TowHook_Tethered_Base_Pose})
  ]
 }
 Right Loop {
  path=SecondMovementMode_FinalPoseOverride.AnimGraphNode_StateMachine_1.Tethered Shift SM.AnimStateNode_35.Right Loop
  nodes [
   Sequence Player 'TowHook_Tethered_Reposition_Right'() {bLoopAnimation=false, sequence=&/TowHook_Tethered_Reposition_Right}
   Output Animation Pose(Result=Sequence Player 'TowHook_Tethered_Reposition_Right_Loop'() {sequence=&/TowHook_Tethered_Reposition_Right_Loop})
  ]
 }
 Left Loop {
  path=SecondMovementMode_FinalPoseOverride.AnimGraphNode_StateMachine_1.Tethered Shift SM.AnimStateNode_36.Left Loop
  nodes [
   Sequence Player 'TowHook_Tethered_Reposition_Left'() {bLoopAnimation=false, sequence=&/TowHook_Tethered_Reposition_Left}
   Output Animation Pose(Result=Sequence Player 'TowHook_Tethered_Reposition_Left_Loop'() {sequence=&/TowHook_Tethered_Reposition_Left_Loop})
  ]
 }
 Left to Right {
  path=SecondMovementMode_FinalPoseOverride.AnimGraphNode_StateMachine_1.Tethered Shift SM.AnimStateNode_37.Left to Right
  nodes [Output Animation Pose(Result=Sequence Player 'TowHook_Tethered_Reposition_LefttoRight'() {bLoopAnimation=false, sequence=&/TowHook_Tethered_Reposition_LefttoRight})]
 }
 Right to Left {
  path=SecondMovementMode_FinalPoseOverride.AnimGraphNode_StateMachine_1.Tethered Shift SM.AnimStateNode_38.Right to Left
  nodes [Output Animation Pose(Result=Sequence Player 'TowHook_Tethered_Reposition_RighttoLeft'() {bLoopAnimation=false, sequence=&/TowHook_Tethered_Reposition_RighttoLeft})]
 }
 Forward Loop {
  path=SecondMovementMode_FinalPoseOverride.AnimGraphNode_StateMachine_1.Tethered Shift SM.AnimStateNode_39.Forward Loop
  nodes [
   Sequence Player 'TowHook_Tethered_Reposition_Forward'() {bLoopAnimation=false, sequence=&/TowHook_Tethered_Reposition_Forward}
   Output Animation Pose(Result=Sequence Player 'TowHook_Tethered_Reposition_Forward_Loop'() {sequence=&/TowHook_Tethered_Reposition_Forward_Loop})
  ]
 }
 Backward Loop {
  path=SecondMovementMode_FinalPoseOverride.AnimGraphNode_StateMachine_1.Tethered Shift SM.AnimStateNode_40.Backward Loop
  nodes [
   Sequence Player 'TowHook_Tethered_Reposition_Backward'() {bLoopAnimation=false, sequence=&/TowHook_Tethered_Reposition_Backward}
   Output Animation Pose(Result=Sequence Player 'TowHook_Tethered_Reposition_Backward_Loop'() {sequence=&/TowHook_Tethered_Reposition_Backward_Loop})
  ]
 }
 Backward Start {
  path=SecondMovementMode_FinalPoseOverride.AnimGraphNode_StateMachine_1.Tethered Shift SM.AnimStateNode_41.Backward Start
  nodes [Output Animation Pose(Result=Sequence Player 'TowHook_Tethered_Reposition_Backward_Start'() {bLoopAnimation=false, sequence=&/TowHook_Tethered_Reposition_Backward_Start})]
 }
 Backward End {
  path=SecondMovementMode_FinalPoseOverride.AnimGraphNode_StateMachine_1.Tethered Shift SM.AnimStateNode_42.Backward End
  nodes [Output Animation Pose(Result=Sequence Player 'TowHook_Tethered_Reposition_Backward_End'() {bLoopAnimation=false, sequence=&/TowHook_Tethered_Reposition_Backward_End})]
 }
 Forward End {
  path=SecondMovementMode_FinalPoseOverride.AnimGraphNode_StateMachine_1.Tethered Shift SM.AnimStateNode_43.Forward End
  nodes [Output Animation Pose(Result=Sequence Player 'TowHook_Tethered_Reposition_Forward_End'() {bLoopAnimation=false, sequence=&/TowHook_Tethered_Reposition_Forward_End})]
 }
 Forward Start {
  path=SecondMovementMode_FinalPoseOverride.AnimGraphNode_StateMachine_1.Tethered Shift SM.AnimStateNode_44.Forward Start
  nodes [Output Animation Pose(Result=Sequence Player 'TowHook_Tethered_Reposition_Forward_Start'() {bLoopAnimation=false, sequence=&/TowHook_Tethered_Reposition_Forward_Start})]
 }
 Right Start {
  path=SecondMovementMode_FinalPoseOverride.AnimGraphNode_StateMachine_1.Tethered Shift SM.AnimStateNode_45.Right Start
  nodes [Output Animation Pose(Result=Sequence Player 'TowHook_Tethered_Reposition_Right_Start'() {bLoopAnimation=false, sequence=&/TowHook_Tethered_Reposition_Right_Start})]
 }
 Right End {
  path=SecondMovementMode_FinalPoseOverride.AnimGraphNode_StateMachine_1.Tethered Shift SM.AnimStateNode_46.Right End
  nodes [Output Animation Pose(Result=Sequence Player 'TowHook_Tethered_Reposition_Right_End'() {bLoopAnimation=false, sequence=&/TowHook_Tethered_Reposition_Right_End})]
 }
 Leftt End {
  path=SecondMovementMode_FinalPoseOverride.AnimGraphNode_StateMachine_1.Tethered Shift SM.AnimStateNode_47.Leftt End
  nodes [Output Animation Pose(Result=Sequence Player 'TowHook_Tethered_Reposition_Left_End'() {bLoopAnimation=false, sequence=&/TowHook_Tethered_Reposition_Left_End})]
 }
 Left Start {
  path=SecondMovementMode_FinalPoseOverride.AnimGraphNode_StateMachine_1.Tethered Shift SM.AnimStateNode_48.Left Start
  nodes [Output Animation Pose(Result=Sequence Player 'TowHook_Tethered_Reposition_Left_Start'() {bLoopAnimation=false, sequence=&/TowHook_Tethered_Reposition_Left_Start})]
 }
 BacktoRight {
  path=SecondMovementMode_FinalPoseOverride.AnimGraphNode_StateMachine_1.Tethered Shift SM.AnimStateNode_49.BacktoRight
  nodes [Output Animation Pose(Result=Sequence Player 'TowHook_Tethered_Reposition_BacktoRight'() {bLoopAnimation=false, sequence=&/TowHook_Tethered_Reposition_BacktoRight})]
 }
 BacktoLeft {
  path=SecondMovementMode_FinalPoseOverride.AnimGraphNode_StateMachine_1.Tethered Shift SM.AnimStateNode_50.BacktoLeft
  nodes [Output Animation Pose(Result=Sequence Player 'TowHook_Tethered_Reposition_BacktoLeft'() {bLoopAnimation=false, sequence=&/TowHook_Tethered_Reposition_BacktoLeft})]
 }
 Transition {
  path=SecondMovementMode_FinalPoseOverride.AnimGraphNode_StateMachine_1.Tethered Shift SM.AnimStateTransitionNode_0.Transition
  nodes [Result(bCanEnterTransition=Get bIsForwardLean() -> (bIsForwardLean:bool))]
 }
 Transition {
  path=SecondMovementMode_FinalPoseOverride.AnimGraphNode_StateMachine_1.Tethered Shift SM.AnimStateTransitionNode_1.Transition
  nodes [Result(bCanEnterTransition=Get bIsRightLean() -> (bIsRightLean:bool))]
 }
 Transition {
  path=SecondMovementMode_FinalPoseOverride.AnimGraphNode_StateMachine_1.Tethered Shift SM.AnimStateTransitionNode_2.Transition
  nodes [Result(bCanEnterTransition=Get bIsLeftLean() -> (bIsLeftLean:bool))]
 }
 AnimationTransitionGraph_0 {
  path=SecondMovementMode_FinalPoseOverride.AnimGraphNode_StateMachine_1.Tethered Shift SM.AnimStateTransitionNode_64.AnimationTransitionGraph_0
  nodes [Result(bCanEnterTransition=NOT Boolean(A=Get bIsAnyLean() -> (bIsAnyLean:bool)))]
 }
 AnimationTransitionGraph_0 {
  path=SecondMovementMode_FinalPoseOverride.AnimGraphNode_StateMachine_1.Tethered Shift SM.AnimStateTransitionNode_65.AnimationTransitionGraph_0
  nodes [Result()]
 }
 AnimationTransitionGraph_0 {
  path=SecondMovementMode_FinalPoseOverride.AnimGraphNode_StateMachine_1.Tethered Shift SM.AnimStateTransitionNode_66.AnimationTransitionGraph_0
  nodes [Result(bCanEnterTransition=Get bIsRightLean())]
 }
 AnimationTransitionGraph_0 {
  path=SecondMovementMode_FinalPoseOverride.AnimGraphNode_StateMachine_1.Tethered Shift SM.AnimStateTransitionNode_67.AnimationTransitionGraph_0
  nodes [Result()]
 }
 AnimationTransitionGraph_0 {
  path=SecondMovementMode_FinalPoseOverride.AnimGraphNode_StateMachine_1.Tethered Shift SM.AnimStateTransitionNode_68.AnimationTransitionGraph_0
  nodes [Result(bCanEnterTransition=NOT Boolean(A=Get bIsAnyLean()))]
 }
 AnimationTransitionGraph_0 {
  path=SecondMovementMode_FinalPoseOverride.AnimGraphNode_StateMachine_1.Tethered Shift SM.AnimStateTransitionNode_71.AnimationTransitionGraph_0
  nodes [Result()]
 }
 AnimationTransitionGraph_0 {
  path=SecondMovementMode_FinalPoseOverride.AnimGraphNode_StateMachine_1.Tethered Shift SM.AnimStateTransitionNode_72.AnimationTransitionGraph_0
  nodes [Result(bCanEnterTransition=NOT Boolean(A=Get bIsAnyLean()))]
 }
 AnimationTransitionGraph_0 {
  path=SecondMovementMode_FinalPoseOverride.AnimGraphNode_StateMachine_1.Tethered Shift SM.AnimStateTransitionNode_74.AnimationTransitionGraph_0
  nodes [Result(bCanEnterTransition=NOT Boolean(A=Get bIsAnyLean()))]
 }
 AnimationTransitionGraph_0 {
  path=SecondMovementMode_FinalPoseOverride.AnimGraphNode_StateMachine_1.Tethered Shift SM.AnimStateTransitionNode_76.AnimationTransitionGraph_0
  nodes [Result(bCanEnterTransition=Get bIsLeftLean())]
 }
 AnimationTransitionGraph_0 {
  path=SecondMovementMode_FinalPoseOverride.AnimGraphNode_StateMachine_1.Tethered Shift SM.AnimStateTransitionNode_77.AnimationTransitionGraph_0
  nodes [Result()]
 }
 AnimationTransitionGraph_0 {
  path=SecondMovementMode_FinalPoseOverride.AnimGraphNode_StateMachine_1.Tethered Shift SM.AnimStateTransitionNode_78.AnimationTransitionGraph_0
  nodes [Result()]
 }
 AnimationTransitionGraph_0 {
  path=SecondMovementMode_FinalPoseOverride.AnimGraphNode_StateMachine_1.Tethered Shift SM.AnimStateTransitionNode_79.AnimationTransitionGraph_0
  nodes [Result()]
 }
 AnimationTransitionGraph_0 {
  path=SecondMovementMode_FinalPoseOverride.AnimGraphNode_StateMachine_1.Tethered Shift SM.AnimStateTransitionNode_87.AnimationTransitionGraph_0
  nodes [Result()]
 }
 AnimationTransitionGraph_0 {
  path=SecondMovementMode_FinalPoseOverride.AnimGraphNode_StateMachine_1.Tethered Shift SM.AnimStateTransitionNode_89.AnimationTransitionGraph_0
  nodes [Result()]
 }
 Transition {
  path=SecondMovementMode_FinalPoseOverride.AnimGraphNode_StateMachine_1.Tethered Shift SM.AnimStateTransitionNode_90.Transition
  nodes [Result()]
 }
 Transition {
  path=SecondMovementMode_FinalPoseOverride.AnimGraphNode_StateMachine_1.Tethered Shift SM.AnimStateTransitionNode_91.Transition
  nodes [Result(bCanEnterTransition=Get bIsRightLean())]
 }
 Transition {
  path=SecondMovementMode_FinalPoseOverride.AnimGraphNode_StateMachine_1.Tethered Shift SM.AnimStateTransitionNode_92.Transition
  nodes [Result(bCanEnterTransition=Get bIsLeftLean())]
 }
 Transition {
  path=SecondMovementMode_FinalPoseOverride.AnimGraphNode_StateMachine_1.Tethered Shift SM.AnimStateTransitionNode_93.Transition
  nodes [Result(bCanEnterTransition=Get bIsBackwardLean() -> (bIsBackwardLean:bool))]
 }
 Yank SM {
  entry=Yank FWD Weak
  path=SecondMovementMode_FinalPoseOverride.AnimGraphNode_StateMachine_2.Yank SM
  ids=id
  nodes [
   To Yank Over() {
    is=title AnimStateAliasNode
    also {aliases [Yank BCK Strong, Yank BCK Weak, Yank FWD Strong, Yank FWD Weak, Yank LT Strong, Yank LT Weak, Yank RT Strong, Yank RT Weak]}
   }
   AnimStateConduitNode_0 = Direction?() {is=title AnimStateConduitNode}
   AnimStateConduitNode_1 = FWD Strength?() {is=title AnimStateConduitNode}
   AnimStateConduitNode_2 = BCK Strength?() {is=title AnimStateConduitNode}
   AnimStateConduitNode_3 = LT Strength?() {is=title AnimStateConduitNode}
   AnimStateConduitNode_4 = RT Strength?() {is=title AnimStateConduitNode}
   AnimStateNode_0 = Yank FWD Strong() {is=title AnimStateNode}
   AnimStateNode_1 = Yank BCK Strong() {is=title AnimStateNode}
   AnimStateNode_2 = Yank RT Strong() {is=title AnimStateNode}
   AnimStateNode_3 = Yank LT Strong() {is=title AnimStateNode}
   AnimStateNode_4 = Yank Over() {is=title AnimStateNode}
   AnimStateNode_5 = Yank FWD Weak() {is=title AnimStateNode}
   AnimStateNode_6 = Yank RT Weak() {is=title AnimStateNode}
   AnimStateNode_7 = Yank BCK Weak() {is=title AnimStateNode}
   AnimStateNode_8 = Yank LT Weak() {is=title AnimStateNode}
   AnimStateTransitionNode_10 = AnimStateTransitionNode()
   AnimStateTransitionNode_11 = AnimStateTransitionNode()
   AnimStateTransitionNode_12 = AnimStateTransitionNode()
   AnimStateTransitionNode_13 = AnimStateTransitionNode()
   AnimStateTransitionNode_14 = AnimStateTransitionNode()
   AnimStateTransitionNode_15 = AnimStateTransitionNode()
   AnimStateTransitionNode_16 = AnimStateTransitionNode()
   AnimStateTransitionNode_17 = AnimStateTransitionNode()
   AnimStateTransitionNode_18 = AnimStateTransitionNode()
   AnimStateTransitionNode_19 = AnimStateTransitionNode()
   AnimStateTransitionNode_20 = AnimStateTransitionNode()
   AnimStateTransitionNode_21 = AnimStateTransitionNode()
   AnimStateTransitionNode_4 = AnimStateTransitionNode()
   AnimStateTransitionNode_5 = AnimStateTransitionNode() {bAutomaticRuleBasedOnSequencePlayerInState=true, crossfadeDuration=0.4}
  ]
  links [
   {
    dst=Yank BCK Strong
    rule=SecondMovementMode_FinalPoseOverride.AnimGraphNode_StateMachine_2.Yank SM.AnimStateTransitionNode_16.Transition
    src=BCK Strength?
    via=AnimStateTransitionNode_16
   }
   {
    dst=Yank BCK Weak
    rule=SecondMovementMode_FinalPoseOverride.AnimGraphNode_StateMachine_2.Yank SM.AnimStateTransitionNode_17.Transition
    src=BCK Strength?
    via=AnimStateTransitionNode_17
   }
   {
    dst=BCK Strength?
    rule=SecondMovementMode_FinalPoseOverride.AnimGraphNode_StateMachine_2.Yank SM.AnimStateTransitionNode_10.Transition
    src=Direction?
    via=AnimStateTransitionNode_10
   }
   {
    dst=FWD Strength?
    rule=SecondMovementMode_FinalPoseOverride.AnimGraphNode_StateMachine_2.Yank SM.AnimStateTransitionNode_15.Transition
    src=Direction?
    via=AnimStateTransitionNode_15
   }
   {
    dst=LT Strength?
    rule=SecondMovementMode_FinalPoseOverride.AnimGraphNode_StateMachine_2.Yank SM.AnimStateTransitionNode_14.Transition
    src=Direction?
    via=AnimStateTransitionNode_14
   }
   {
    dst=RT Strength?
    rule=SecondMovementMode_FinalPoseOverride.AnimGraphNode_StateMachine_2.Yank SM.AnimStateTransitionNode_11.Transition
    src=Direction?
    via=AnimStateTransitionNode_11
   }
   {
    dst=Yank FWD Strong
    rule=SecondMovementMode_FinalPoseOverride.AnimGraphNode_StateMachine_2.Yank SM.AnimStateTransitionNode_18.Transition
    src=FWD Strength?
    via=AnimStateTransitionNode_18
   }
   {
    dst=Yank FWD Weak
    rule=SecondMovementMode_FinalPoseOverride.AnimGraphNode_StateMachine_2.Yank SM.AnimStateTransitionNode_19.Transition
    src=FWD Strength?
    via=AnimStateTransitionNode_19
   }
   {
    dst=Yank LT Strong
    rule=SecondMovementMode_FinalPoseOverride.AnimGraphNode_StateMachine_2.Yank SM.AnimStateTransitionNode_21.Transition
    src=LT Strength?
    via=AnimStateTransitionNode_21
   }
   {
    dst=Yank LT Weak
    rule=SecondMovementMode_FinalPoseOverride.AnimGraphNode_StateMachine_2.Yank SM.AnimStateTransitionNode_20.Transition
    src=LT Strength?
    via=AnimStateTransitionNode_20
   }
   {
    dst=Yank RT Strong
    rule=SecondMovementMode_FinalPoseOverride.AnimGraphNode_StateMachine_2.Yank SM.AnimStateTransitionNode_13.Transition
    src=RT Strength?
    via=AnimStateTransitionNode_13
   }
   {
    dst=Yank RT Weak
    rule=SecondMovementMode_FinalPoseOverride.AnimGraphNode_StateMachine_2.Yank SM.AnimStateTransitionNode_12.Transition
    src=RT Strength?
    via=AnimStateTransitionNode_12
   }
   {
    dst=Yank Over
    rule=SecondMovementMode_FinalPoseOverride.AnimGraphNode_StateMachine_2.Yank SM.AnimStateTransitionNode_5.Transition
    src=To Yank Over
    via=AnimStateTransitionNode_5
   }
   {
    dst=Direction?
    rule=SecondMovementMode_FinalPoseOverride.AnimGraphNode_StateMachine_2.Yank SM.AnimStateTransitionNode_4.Transition
    src=Yank FWD Weak
    via=AnimStateTransitionNode_4
   }
  ]
 }
 Direction? {
  path=SecondMovementMode_FinalPoseOverride.AnimGraphNode_StateMachine_2.Yank SM.AnimStateConduitNode_0.Direction?
  nodes [Result(bCanEnterTransition=true)]
 }
 FWD Strength? {
  path=SecondMovementMode_FinalPoseOverride.AnimGraphNode_StateMachine_2.Yank SM.AnimStateConduitNode_1.FWD Strength?
  nodes [Result(bCanEnterTransition=true)]
 }
 BCK Strength? {
  path=SecondMovementMode_FinalPoseOverride.AnimGraphNode_StateMachine_2.Yank SM.AnimStateConduitNode_2.BCK Strength?
  nodes [Result(bCanEnterTransition=true)]
 }
 LT Strength? {
  path=SecondMovementMode_FinalPoseOverride.AnimGraphNode_StateMachine_2.Yank SM.AnimStateConduitNode_3.LT Strength?
  nodes [Result(bCanEnterTransition=true)]
 }
 RT Strength? {
  path=SecondMovementMode_FinalPoseOverride.AnimGraphNode_StateMachine_2.Yank SM.AnimStateConduitNode_4.RT Strength?
  nodes [Result(bCanEnterTransition=true)]
 }
 Yank FWD Strong {
  path=SecondMovementMode_FinalPoseOverride.AnimGraphNode_StateMachine_2.Yank SM.AnimStateNode_0.Yank FWD Strong
  nodes [
   n1 = Sequence Player 'TowHook_Tethered_Yank_Strong_Forward'() {bLoopAnimation=false, groupName=yank, method=SyncGroup, sequence=&/TowHook_Tethered_Yank_Strong_Forward}
   n2 = Sequence Player 'TowHook_Tethered_Yank_Air_Strong_Forward'() {bLoopAnimation=false, groupName=yank, method=SyncGroup, sequence=&/TowHook_Tethered_Yank_Air_Strong_Forward}
   Output Animation Pose(Result=Two Way Blend(A=@n1, B=@n2, Alpha=0) {blendNode {alphaBoolBlend {blendOption=Linear}, alphaScaleBiasClamp {clampMax=0.99, clampMin=0.01}}})
  ]
 }
 Yank BCK Strong {
  path=SecondMovementMode_FinalPoseOverride.AnimGraphNode_StateMachine_2.Yank SM.AnimStateNode_1.Yank BCK Strong
  nodes [
   n1 = Sequence Player 'TowHook_Tethered_Yank_Strong_Backward'() {bLoopAnimation=false, groupName=yank, method=SyncGroup, sequence=&/TowHook_Tethered_Yank_Strong_Backward}
   n2 = Sequence Player 'TowHook_Tethered_Yank_Air_Strong_Backward'() {bLoopAnimation=false, groupName=yank, method=SyncGroup, sequence=&/TowHook_Tethered_Yank_Air_Strong_Backward}
   Output Animation Pose(Result=Two Way Blend(A=@n1, B=@n2, Alpha=0) {blendNode {alphaBoolBlend {blendOption=Linear}, alphaScaleBiasClamp {clampMax=0.99, clampMin=0.01}}})
  ]
 }
 Yank RT Strong {
  path=SecondMovementMode_FinalPoseOverride.AnimGraphNode_StateMachine_2.Yank SM.AnimStateNode_2.Yank RT Strong
  nodes [
   n1 = Sequence Player 'TowHook_Tethered_Yank_Strong_Right'() {bLoopAnimation=false, groupName=yank, method=SyncGroup, sequence=&/TowHook_Tethered_Yank_Strong_Right}
   n2 = Sequence Player 'TowHook_Tethered_Yank_Air_Strong_Right'() {bLoopAnimation=false, groupName=yank, method=SyncGroup, sequence=&/TowHook_Tethered_Yank_Air_Strong_Right}
   Output Animation Pose(Result=Two Way Blend(A=@n1, B=@n2, Alpha=0) {blendNode {alphaBoolBlend {blendOption=Linear}, alphaScaleBiasClamp {clampMax=0.99, clampMin=0.01}}})
  ]
 }
 Yank LT Strong {
  path=SecondMovementMode_FinalPoseOverride.AnimGraphNode_StateMachine_2.Yank SM.AnimStateNode_3.Yank LT Strong
  nodes [
   n1 = Sequence Player 'TowHook_Tethered_Yank_Strong_Left'() {bLoopAnimation=false, groupName=yank, method=SyncGroup, sequence=&/TowHook_Tethered_Yank_Strong_Left}
   n2 = Sequence Player 'TowHook_Tethered_Yank_Air_Strong_Left'() {bLoopAnimation=false, groupName=yank, method=SyncGroup, sequence=&/TowHook_Tethered_Yank_Air_Strong_Left}
   Output Animation Pose(Result=Two Way Blend(A=@n1, B=@n2, Alpha=0) {blendNode {alphaBoolBlend {blendOption=Linear}, alphaScaleBiasClamp {clampMax=0.99, clampMin=0.01}}})
  ]
 }
 Yank Over {
  path=SecondMovementMode_FinalPoseOverride.AnimGraphNode_StateMachine_2.Yank SM.AnimStateNode_4.Yank Over
  nodes [Output Animation Pose(Result=Modify Curve(SourcePose=Use cached pose 'Pre Yank Pose'(), CurveValues_0=1, Alpha=1) {curveNames [UserCurve02], curveValues [0]})]
 }
 Yank FWD Weak {
  path=SecondMovementMode_FinalPoseOverride.AnimGraphNode_StateMachine_2.Yank SM.AnimStateNode_5.Yank FWD Weak
  nodes [
   n1 = Sequence Player 'TowHook_Tethered_Yank_Forward'() {bLoopAnimation=false, groupName=yank, method=SyncGroup, sequence=&/TowHook_Tethered_Yank_Forward}
   n2 = Sequence Player 'TowHook_Tethered_Yank_Air_Forward'() {bLoopAnimation=false, groupName=yank, method=SyncGroup, sequence=&/TowHook_Tethered_Yank_Air_Forward}
   Output Animation Pose(Result=Two Way Blend(A=@n1, B=@n2, Alpha=0) {blendNode {alphaBoolBlend {blendOption=Linear}, alphaScaleBiasClamp {clampMax=0.99, clampMin=0.01}}})
  ]
 }
 Yank RT Weak {
  path=SecondMovementMode_FinalPoseOverride.AnimGraphNode_StateMachine_2.Yank SM.AnimStateNode_6.Yank RT Weak
  nodes [
   n1 = Sequence Player 'TowHook_Tethered_Yank_Right'() {bLoopAnimation=false, groupName=yank, method=SyncGroup, sequence=&/TowHook_Tethered_Yank_Right}
   n2 = Sequence Player 'TowHook_Tethered_Yank_Air_Right'() {bLoopAnimation=false, groupName=yank, method=SyncGroup, sequence=&/TowHook_Tethered_Yank_Air_Right}
   Output Animation Pose(Result=Two Way Blend(A=@n1, B=@n2, Alpha=0) {blendNode {alphaBoolBlend {blendOption=Linear}, alphaScaleBiasClamp {clampMax=0.99, clampMin=0.01}}})
  ]
 }
 Yank BCK Weak {
  path=SecondMovementMode_FinalPoseOverride.AnimGraphNode_StateMachine_2.Yank SM.AnimStateNode_7.Yank BCK Weak
  nodes [
   n1 = Sequence Player 'TowHook_Tethered_Yank_Backward'() {bLoopAnimation=false, groupName=yank, method=SyncGroup, sequence=&/TowHook_Tethered_Yank_Backward}
   n2 = Sequence Player 'TowHook_Tethered_Yank_Air_Backward'() {bLoopAnimation=false, groupName=yank, method=SyncGroup, sequence=&/TowHook_Tethered_Yank_Air_Backward}
   Output Animation Pose(Result=Two Way Blend(A=@n1, B=@n2, Alpha=0) {blendNode {alphaBoolBlend {blendOption=Linear}, alphaScaleBiasClamp {clampMax=0.99, clampMin=0.01}}})
  ]
 }
 Yank LT Weak {
  path=SecondMovementMode_FinalPoseOverride.AnimGraphNode_StateMachine_2.Yank SM.AnimStateNode_8.Yank LT Weak
  nodes [
   n1 = Sequence Player 'TowHook_Tethered_Yank_Air_Left'() {bLoopAnimation=false, groupName=yank, method=SyncGroup, sequence=&/TowHook_Tethered_Yank_Air_Left}
   n2 = Sequence Player 'TowHook_Tethered_Yank_Left'() {bLoopAnimation=false, groupName=yank, method=SyncGroup, sequence=&/TowHook_Tethered_Yank_Left}
   Output Animation Pose(Result=Two Way Blend(A=@n2, B=@n1, Alpha=0) {blendNode {alphaBoolBlend {blendOption=Linear}, alphaScaleBiasClamp {clampMax=0.99, clampMin=0.01}}})
  ]
 }
 Transition {
  path=SecondMovementMode_FinalPoseOverride.AnimGraphNode_StateMachine_2.Yank SM.AnimStateTransitionNode_10.Transition
  nodes [Result(bCanEnterTransition=Get bIsBack() -> (bIsBack:bool))]
 }
 Transition {
  path=SecondMovementMode_FinalPoseOverride.AnimGraphNode_StateMachine_2.Yank SM.AnimStateTransitionNode_11.Transition
  nodes [Result(bCanEnterTransition=Get bIsRight() -> (bIsRight:bool))]
 }
 Transition {
  path=SecondMovementMode_FinalPoseOverride.AnimGraphNode_StateMachine_2.Yank SM.AnimStateTransitionNode_12.Transition
  nodes [Result(bCanEnterTransition=Get bIsYankWeak() -> (bIsYankWeak:bool))]
 }
 Transition {
  path=SecondMovementMode_FinalPoseOverride.AnimGraphNode_StateMachine_2.Yank SM.AnimStateTransitionNode_13.Transition
  nodes [Result(bCanEnterTransition=Get bIsYankStrong() -> (bIsYankStrong:bool))]
 }
 Transition {
  path=SecondMovementMode_FinalPoseOverride.AnimGraphNode_StateMachine_2.Yank SM.AnimStateTransitionNode_14.Transition
  nodes [Result(bCanEnterTransition=Get bIsLeft() -> (bIsLeft:bool))]
 }
 Transition {
  path=SecondMovementMode_FinalPoseOverride.AnimGraphNode_StateMachine_2.Yank SM.AnimStateTransitionNode_15.Transition
  nodes [Result(bCanEnterTransition=Get bIsForward() -> (bIsForward:bool))]
 }
 Transition {
  path=SecondMovementMode_FinalPoseOverride.AnimGraphNode_StateMachine_2.Yank SM.AnimStateTransitionNode_16.Transition
  nodes [Result(bCanEnterTransition=Get bIsYankStrong())]
 }
 Transition {
  path=SecondMovementMode_FinalPoseOverride.AnimGraphNode_StateMachine_2.Yank SM.AnimStateTransitionNode_17.Transition
  nodes [Result(bCanEnterTransition=Get bIsYankWeak())]
 }
 Transition {
  path=SecondMovementMode_FinalPoseOverride.AnimGraphNode_StateMachine_2.Yank SM.AnimStateTransitionNode_18.Transition
  nodes [Result(bCanEnterTransition=Get bIsYankStrong())]
 }
 Transition {
  path=SecondMovementMode_FinalPoseOverride.AnimGraphNode_StateMachine_2.Yank SM.AnimStateTransitionNode_19.Transition
  nodes [Result(bCanEnterTransition=Get bIsYankWeak())]
 }
 Transition {
  path=SecondMovementMode_FinalPoseOverride.AnimGraphNode_StateMachine_2.Yank SM.AnimStateTransitionNode_20.Transition
  nodes [Result(bCanEnterTransition=Get bIsYankWeak())]
 }
 Transition {
  path=SecondMovementMode_FinalPoseOverride.AnimGraphNode_StateMachine_2.Yank SM.AnimStateTransitionNode_21.Transition
  nodes [Result(bCanEnterTransition=Get bIsYankStrong())]
 }
 Transition {
  path=SecondMovementMode_FinalPoseOverride.AnimGraphNode_StateMachine_2.Yank SM.AnimStateTransitionNode_4.Transition
  nodes [Result(bCanEnterTransition=Get bIsBackOrLeftOrRight() -> (bIsBackOrLeftOrRight:bool))]
 }
 Transition {path=SecondMovementMode_FinalPoseOverride.AnimGraphNode_StateMachine_2.Yank SM.AnimStateTransitionNode_5.Transition, nodes [Result()]}
 Attach SM {
  entry=Attach Stationary
  path=SecondMovementMode_FinalPoseOverride.AnimGraphNode_StateMachine_3.Attach SM
  ids=id
  nodes [
   AnimStateNode_0 = Pass Through Out() {is=title AnimStateNode}
   AnimStateNode_1 = Attach Slow() {
    is=title AnimStateNode
    stateEntered {guid=#25}
    stateFullyBlended {guid=#26}
    stateLeft {guid=#27}
   }
   AnimStateNode_2 = Attach Fast() {
    is=title AnimStateNode
    stateEntered {guid=#28}
    stateFullyBlended {guid=#29}
    stateLeft {guid=#30}
   }
   AnimStateNode_5 = Attach Stationary() {
    is=title AnimStateNode
    stateEntered {guid=#31}
    stateFullyBlended {guid=#32}
    stateLeft {guid=#33}
   }
   AnimStateTransitionNode_0 = AnimStateTransitionNode() {
    priorityOrder=10
    transitionEnd {guid=#34}
    transitionInterrupt {guid=#35}
    transitionStart {guid=#36}
   }
   AnimStateTransitionNode_1 = AnimStateTransitionNode() {
    priorityOrder=10
    transitionEnd {guid=#37}
    transitionInterrupt {guid=#38}
    transitionStart {guid=#39}
   }
   AnimStateTransitionNode_3 = AnimStateTransitionNode() {bAutomaticRuleBasedOnSequencePlayerInState=true}
   AnimStateTransitionNode_4 = AnimStateTransitionNode() {bAutomaticRuleBasedOnSequencePlayerInState=true}
  ]
  links [
   {
    dst=Pass Through Out
    rule=SecondMovementMode_FinalPoseOverride.AnimGraphNode_StateMachine_3.Attach SM.AnimStateTransitionNode_4.Transition
    src=Attach Fast
    via=AnimStateTransitionNode_4
   }
   {
    dst=Pass Through Out
    rule=SecondMovementMode_FinalPoseOverride.AnimGraphNode_StateMachine_3.Attach SM.AnimStateTransitionNode_3.Transition
    src=Attach Slow
    via=AnimStateTransitionNode_3
   }
   {
    dst=Attach Fast
    rule=SecondMovementMode_FinalPoseOverride.AnimGraphNode_StateMachine_3.Attach SM.AnimStateTransitionNode_1.Transition
    src=Attach Stationary
    via=AnimStateTransitionNode_1
   }
   {
    dst=Attach Slow
    rule=SecondMovementMode_FinalPoseOverride.AnimGraphNode_StateMachine_3.Attach SM.AnimStateTransitionNode_0.Transition
    src=Attach Stationary
    via=AnimStateTransitionNode_0
   }
  ]
 }
 Pass Through Out {
  path=SecondMovementMode_FinalPoseOverride.AnimGraphNode_StateMachine_3.Attach SM.AnimStateNode_0.Pass Through Out
  nodes [
   Output Animation Pose(Result=Use cached pose 'PreAddLean Pose'())
   Use cached pose 'FinalPose Cached'()
  ]
 }
 Attach Slow {
  path=SecondMovementMode_FinalPoseOverride.AnimGraphNode_StateMachine_3.Attach SM.AnimStateNode_1.Attach Slow
  nodes [
   n1 = Sequence Player 'TowHook_Tethered_Attach_Slow'() {bLoopAnimation=false, groupName=Attach, method=SyncGroup, sequence=&/TowHook_Tethered_Attach_Slow}
   n2 = Sequence Player 'TowHook_Tethered_Attach_Slow'() {bLoopAnimation=false, groupName=Attach, method=SyncGroup, sequence=&TowHook_Tethered_Attach_Slow}
   n3 = Use cached pose 'PreAddLean Pose'()
   Output Animation Pose(Result=Layered blend per bone(BasePose=Two Way Blend(A=@n3, B=@n2, Alpha=0) {blendNode {alphaBoolBlend {blendOption=Linear}}}, BlendPoses_0=@n1) {layerSetup [{branchFilters [{boneName=attach}]}]})
   Use cached pose 'FinalPose Cached'()
  ]
 }
 Attach Fast {
  path=SecondMovementMode_FinalPoseOverride.AnimGraphNode_StateMachine_3.Attach SM.AnimStateNode_2.Attach Fast
  nodes [
   n1 = Sequence Player 'TowHook_Tethered_Attach_Fast_Air'() {bLoopAnimation=false, groupName=Attach, method=SyncGroup, sequence=&/TowHook_Tethered_Attach_Fast_Air}
   n2 = Sequence Player 'TowHook_Tethered_Attach_Fast_Air'() {bLoopAnimation=false, groupName=Attach, method=SyncGroup, sequence=&TowHook_Tethered_Attach_Fast_Air}
   n3 = Two Way Blend(A=Use cached pose 'PreAddLean Pose'(), B=@n2, Alpha=0) {blendNode {alphaBoolBlend {blendOption=Linear}}}
   n4 = Sequence Player 'TowHook_Tethered_Attach_Fast'() {bLoopAnimation=false, groupName=Attach, method=SyncGroup, sequence=&/TowHook_Tethered_Attach_Fast}
   n5 = Sequence Player 'TowHook_Tethered_Attach_Fast'() {bLoopAnimation=false, groupName=Attach, method=SyncGroup, sequence=&TowHook_Tethered_Attach_Fast}
   n6 = Two Way Blend(A=Use cached pose 'PreAddLean Pose'(), B=@n5, Alpha=0) {blendNode {alphaBoolBlend {blendOption=Linear}}}
   Output Animation Pose(Result=Blend Poses by bool(
     BlendPose_0=Layered blend per bone(BasePose=@n3, BlendPoses_0=@n1) {layerSetup [{branchFilters [{boneName=attach}]}]}
     BlendPose_1=Layered blend per bone(BasePose=@n6, BlendPoses_0=@n4) {layerSetup [{branchFilters [{boneName=attach}]}]}
     BlendTime_0=0.15
     BlendTime_1=0.15
    ) {blendType=Linear})
   Use cached pose 'FinalPose Cached'()
   Use cached pose 'FinalPose Cached'()
  ]
 }
 Attach Stationary {
  path=SecondMovementMode_FinalPoseOverride.AnimGraphNode_StateMachine_3.Attach SM.AnimStateNode_5.Attach Stationary
  nodes [
   n1 = Sequence Player 'TowHook_Tethered_Attach_Stationary'() {bLoopAnimation=false, groupName=Attach, method=SyncGroup, sequence=&/TowHook_Tethered_Attach_Stationary}
   Output Animation Pose(Result=Layered blend per bone(BasePose=Use cached pose 'FinalPose Cached'(), BlendPoses_0=@n1) {layerSetup [{branchFilters [{boneName=attach}]}]})
  ]
 }
 Transition {
  path=SecondMovementMode_FinalPoseOverride.AnimGraphNode_StateMachine_3.Attach SM.AnimStateTransitionNode_0.Transition
  nodes [Result(bCanEnterTransition=Get bIsAttachSlow() -> (bIsAttachSlow:bool))]
 }
 Transition {
  path=SecondMovementMode_FinalPoseOverride.AnimGraphNode_StateMachine_3.Attach SM.AnimStateTransitionNode_1.Transition
  nodes [Result(bCanEnterTransition=Get bIsAttachFast() -> (bIsAttachFast:bool))]
 }
 Transition {path=SecondMovementMode_FinalPoseOverride.AnimGraphNode_StateMachine_3.Attach SM.AnimStateTransitionNode_3.Transition, nodes [Result()]}
 Transition {path=SecondMovementMode_FinalPoseOverride.AnimGraphNode_StateMachine_3.Attach SM.AnimStateTransitionNode_4.Transition, nodes [Result()]}
 Mounting SM {
  entry=Pick Direction
  path=SecondMovementMode_FinalPoseOverride.AnimGraphNode_StateMachine_4.Mounting SM
  ids=id
  nodes [
   To Post Mount() {is=title AnimStateAliasNode, bGlobalAlias=true}
   AnimStateConduitNode_0 = Pick Direction() {is=title AnimStateConduitNode}
   AnimStateNode_0 = Post Mount() {is=title AnimStateNode}
   AnimStateNode_1 = Mounting Start() {is=title AnimStateNode}
   AnimStateNode_2 = Mount Forward() {is=title AnimStateNode}
   AnimStateNode_3 = Mount Left() {is=title AnimStateNode}
   AnimStateNode_4 = Mount Right() {is=title AnimStateNode}
   AnimStateNode_5 = Mount Back() {is=title AnimStateNode}
   AnimStateTransitionNode_1 = AnimStateTransitionNode()
   AnimStateTransitionNode_2 = AnimStateTransitionNode()
   AnimStateTransitionNode_3 = AnimStateTransitionNode()
   AnimStateTransitionNode_4 = AnimStateTransitionNode()
   AnimStateTransitionNode_5 = AnimStateTransitionNode() {bAutomaticRuleBasedOnSequencePlayerInState=true}
  ]
  links [
   {
    dst=Mount Back
    rule=SecondMovementMode_FinalPoseOverride.AnimGraphNode_StateMachine_4.Mounting SM.AnimStateTransitionNode_4.Transition
    src=Pick Direction
    via=AnimStateTransitionNode_4
   }
   {
    dst=Mount Forward
    rule=SecondMovementMode_FinalPoseOverride.AnimGraphNode_StateMachine_4.Mounting SM.AnimStateTransitionNode_1.Transition
    src=Pick Direction
    via=AnimStateTransitionNode_1
   }
   {
    dst=Mount Left
    rule=SecondMovementMode_FinalPoseOverride.AnimGraphNode_StateMachine_4.Mounting SM.AnimStateTransitionNode_2.Transition
    src=Pick Direction
    via=AnimStateTransitionNode_2
   }
   {
    dst=Mount Right
    rule=SecondMovementMode_FinalPoseOverride.AnimGraphNode_StateMachine_4.Mounting SM.AnimStateTransitionNode_3.Transition
    src=Pick Direction
    via=AnimStateTransitionNode_3
   }
   {
    dst=Post Mount
    rule=SecondMovementMode_FinalPoseOverride.AnimGraphNode_StateMachine_4.Mounting SM.AnimStateTransitionNode_5.Transition
    src=To Post Mount
    via=AnimStateTransitionNode_5
   }
  ]
 }
 Pick Direction {
  path=SecondMovementMode_FinalPoseOverride.AnimGraphNode_StateMachine_4.Mounting SM.AnimStateConduitNode_0.Pick Direction
  nodes [Result(bCanEnterTransition=true)]
 }
 Post Mount {
  path=SecondMovementMode_FinalPoseOverride.AnimGraphNode_StateMachine_4.Mounting SM.AnimStateNode_0.Post Mount
  nodes [Output Animation Pose(Result=Apply Additive(
     Base=Use cached pose 'FinalPose Cached'()
     Additive="Sequence Player (additive) 'M_JumpLand_ARDS_Additive'"() {bLoopAnimation=false, sequence=/Game/Animation/Game/MainPlayer/Locomotion/Medium/Male/Jump/AssaultRifle/M_JumpLand_ARDS_Additive}
    ) {alphaBoolBlend {blendOption=Linear}})]
 }
 Mounting Start {
  path=SecondMovementMode_FinalPoseOverride.AnimGraphNode_StateMachine_4.Mounting SM.AnimStateNode_1.Mounting Start
  nodes [Output Animation Pose(Result=Blend Poses by bool(
     BlendPose_0=Sequence Player 'TowHook_Tethered_Reel_In_Mount_v2'() {bLoopAnimation=false, sequence=&/TowHook_Tethered_Reel_In_Mount_v2}
     BlendPose_1=Sequence Player 'TowHook_Tethered_Reel_In_Mount_v1'() {bLoopAnimation=false, sequence=&/TowHook_Tethered_Reel_In_Mount_v1}
     BlendTime_0=0.05
     BlendTime_1=0.05
    ) {blendType=Linear})]
 }
 Mount Forward {
  path=SecondMovementMode_FinalPoseOverride.AnimGraphNode_StateMachine_4.Mounting SM.AnimStateNode_2.Mount Forward
  nodes [
   Sequence Player 'TowHook_Mount_Forward'() {bLoopAnimation=false, sequence=/TowHookWeapon/Assets/Animation/MountJumps/TowHook_Mount_Forward}
   Output Animation Pose(Result=Blend Poses by bool(
     BlendPose_0=Sequence Player 'TowHook_Tethered_Reel_In_Mount_v1'() {bLoopAnimation=false, sequence=&TowHook_Tethered_Reel_In_Mount_v1}
     BlendPose_1=Sequence Player 'TowHook_Tethered_Reel_In_Mount_v2'() {bLoopAnimation=false, sequence=&TowHook_Tethered_Reel_In_Mount_v2}
     bActiveValue=Get bIsAltMount() -> (bIsAltMount:bool)
    ) {blendType=Linear})
  ]
 }
 Mount Left {
  path=SecondMovementMode_FinalPoseOverride.AnimGraphNode_StateMachine_4.Mounting SM.AnimStateNode_3.Mount Left
  nodes [Output Animation Pose(Result=Sequence Player 'TowHook_Mount_Left'() {bLoopAnimation=false, sequence=/TowHookWeapon/Assets/Animation/MountJumps/TowHook_Mount_Left})]
 }
 Mount Right {
  path=SecondMovementMode_FinalPoseOverride.AnimGraphNode_StateMachine_4.Mounting SM.AnimStateNode_4.Mount Right
  nodes [Output Animation Pose(Result=Sequence Player 'TowHook_Mount_Right'() {bLoopAnimation=false, sequence=/TowHookWeapon/Assets/Animation/MountJumps/TowHook_Mount_Right})]
 }
 Mount Back {
  path=SecondMovementMode_FinalPoseOverride.AnimGraphNode_StateMachine_4.Mounting SM.AnimStateNode_5.Mount Back
  nodes [Output Animation Pose(Result=Sequence Player 'TowHook_Mount_Back'() {bLoopAnimation=false, sequence=/TowHookWeapon/Assets/Animation/MountJumps/TowHook_Mount_Back})]
 }
 Transition {
  path=SecondMovementMode_FinalPoseOverride.AnimGraphNode_StateMachine_4.Mounting SM.AnimStateTransitionNode_1.Transition
  nodes [Result(bCanEnterTransition=Get bIsForward())]
 }
 Transition {
  path=SecondMovementMode_FinalPoseOverride.AnimGraphNode_StateMachine_4.Mounting SM.AnimStateTransitionNode_2.Transition
  nodes [Result(bCanEnterTransition=Get bIsLeft())]
 }
 Transition {
  path=SecondMovementMode_FinalPoseOverride.AnimGraphNode_StateMachine_4.Mounting SM.AnimStateTransitionNode_3.Transition
  nodes [Result(bCanEnterTransition=Get bIsRight())]
 }
 Transition {
  path=SecondMovementMode_FinalPoseOverride.AnimGraphNode_StateMachine_4.Mounting SM.AnimStateTransitionNode_4.Transition
  nodes [Result(bCanEnterTransition=Get bIsBack())]
 }
 Transition {path=SecondMovementMode_FinalPoseOverride.AnimGraphNode_StateMachine_4.Mounting SM.AnimStateTransitionNode_5.Transition, nodes [Result()]}
]
variables []
