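# Animation Blueprint for the tow hook launcher weapon skeleton, playing montages through DefaultSlot over a reference pose.
# A zero-time Blend Poses by bool picks between the plain slot pose and one with the tow_hook bone scaled to zero, hiding the hook.
# The graph shows nothing driving the bool, and the event graph and variable list are empty.
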AnimBlueprint /TowHookWeapon/Assets/Animation/TowHook_Weapon_AnimBP
target_skeleton=/TowHookWeapon/Assets/Meshes/TowHook_Launcher_Skeleton
graphs [
 AnimGraph {nodes [
   n1 = Local Space Ref Pose()
   n2 = "Transform (Modify) Bone - Bone: tow_hook"(ComponentPose=Local To Component(LocalPose=Slot 'DefaultSlot'(Source=@n1)), Alpha=1, Translation="0,0,0", Rotation="0,0,0", Scale="0,0,0") {scaleMode=BMM_Replace, alphaBoolBlend {blendOption=Linear}, boneToModify {boneName=tow_hook}, scale=<0 0 0>}
   n3 = Local Space Ref Pose()
   Output Pose(Result=Blend Poses by bool(BlendPose_0=Slot 'DefaultSlot'(Source=@n3), BlendPose_1=Component To Local(ComponentPose=@n2), BlendTime_0=0, BlendTime_1=0) {blendType=Linear})
  ]}
 EventGraph {links [], nodes []}
]
variables []
