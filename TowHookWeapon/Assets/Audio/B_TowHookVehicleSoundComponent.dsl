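; Vehicle sound component for the tow hook weapon that plays an attached loop on the vehicle it is added to.
; BeginPlay spawns the loop on the vehicle's root and starts a 0.1-second timer feeding forward speed to its Speed parameter.
; All attached projectiles share one component, so the server tracks their count and calls a fade-out, then destroys the component, when none remain.
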(event EventBeginPlay
  (Utilities|FlowControl|Sequence
    (:then_0
      ;[0 0.1341 0.5521] Cache owning vehicle
      (bind _asfort_athena_vehicle (Utilities|Casting|CastToFortAthenaVehicle (Components|GetOwner)))
      (Variables|Default|SetOwning%20Vehicle _asfort_athena_vehicle))
    (:then_1
      (Utilities|FlowControl|Sequence
        (:then_0
          ;[0 0.1341 0.5521] Start the audio loop
          (bind _owning_vehicle (Variables|Default|GetOwning%20Vehicle)
            (:then
              (bind _returnvalue (Audio|SpawnSoundAttached (Variables|Default|GetTow%20Hook%20Attached%20to%20Vehicle%20Loop) (Variables|/Script/Engine.Actor|GetRootComponent _owning_vehicle)))
              (Utilities|IsValid _returnvalue
                (:Is%20Valid
                  (Variables|Default|SetAudio%20Component _returnvalue)
                  (Utilities|Time|SetTimerbyEvent (delegate Custom|Update) 0.1 true))))
            (:else)))))))

(event Custom|Update
  (CallFunction|UpdateAudioParameters)
  ; Print Debug Attached
  (if false
    (Development|PrintText (Utilities|Text|FormatText "NSLOCTEXT(\"[EA982B8BDF1B97B46CA691ED0F6F5BC2]\", \"5AA8E3B542A5D0E21E8C69B91DE1E261\", \"Vehicle: {vehicle}\\r\\nNumAttached: {num}\")" (Utilities|String|StringToName (Vehicle|GetDisplayName (Variables|Default|GetOwning%20Vehicle) :!class /Script/FortniteGame.FortAthenaVehicle)) (Variables|Default|GetNumber%20of%20Attached%20Projectiles)) true false "(R=0.780357,G=1.000000,B=0.000000,A=1.000000)" 5.0 (Utilities|String|StringToName (Vehicle|GetDisplayName (Variables|Default|GetOwning%20Vehicle) :!class /Script/FortniteGame.FortAthenaVehicle)))))

(event Custom|Modify%20Attach%20Count
  (params (Value :type Integer))
  ;[0 0.1341 0.5521] Server - One component is shared between all projectiles, so we need to keep track of how many projectiles are attached and clean up when there are none left.
  (Utilities|SwitchHasAuthority
    (:Authority
      (bind _output_get (Variables|Default|SetNumber%20of%20Attached%20Projectiles (+ (Variables|Default|GetNumber%20of%20Attached%20Projectiles) Value)))
      (if (<= _output_get 0)
        (CallFunction|FadeOutAndDestroyComponent)))))

(event Custom|Fade%20Out%20and%20Destroy%20Component
  (Utilities|FlowControl|Sequence
    (:then_0
      (bind _audio_component (Variables|Default|GetAudio%20Component)
        (:then
          (Audio|Components|Audio|FadeOut _audio_component 0.5 0.0 "Linear" :!class /Script/Engine.AudioComponent))
        (:else)))
    (:then_1
      (Utilities|FlowControl|Sequence
        (:then_0
          (Utilities|FlowControl|Delay 0.5)
          (Components|DestroyComponent))))))

(fn UpdateAudioParameters
  (bind _audio_component (Variables|Default|GetAudio%20Component)
    (:then
      (Audio|Parameter|SetFloatParameter _audio_component "Speed" (Vehicle|GetForwardSpeed (Variables|Default|GetOwning%20Vehicle)) :!class /Script/AudioExtensions.AudioParameterControllerInterface))
    (:else)))
