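# One-shot mono MetaSound source that plays tow hook weapon foley for leaving aim-down-sights.
# On play, it picks one of two foley waves at random, then rolls a slight random frequency multiplier.
# That multiplier is converted to semitones to pitch-shift the wave player, whose completion fires OnFinished.
# It uses weapon foley class and reload attenuation, with one instance per owner, stopping the oldest.
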MetaSoundSource /TowHookWeapon/Assets/Audio/MSS_TowHookWeapon_ADS_Out
inputs [
 Sounds {type=WaveAsset:Array}
 UE.Source.OnPlay {type=Trigger}
]
interfaces [UE.OutputFormat.Mono:1.0, UE.Source.OneShot:1.0, UE.Source:1.1]
outputs [
 "UE.OutputFormat.Mono.Audio:0" {type=Audio}
 UE.Source.OneShot.OnFinished {type=Trigger}
]
pages [Default {
  input_defaults {
   UE.Source.OnPlay=null
   Sounds [/TowHookWeapon/Assets/Audio/wav/TowHook_Foley_ADS_Out_01, /TowHookWeapon/Assets/Audio/wav/TowHook_Foley_ADS_Out_02]
  }
  nodes [
   ArrayRandomGet_0 = Array.Random Get:WaveAsset:Array(In Array=^Sounds, Next=^UE.Source.OnPlay)
   n1 = UE.RandomFloat(Next=@ArrayRandomGet_0.On Next) {also {defaults {Max=1.1, Min=0.98}}}
   FrequencyMultipliertoSemitone_0 = UE.Frequency Multiplier to Semitone:Float(Frequency Multiplier=@n1.Value)
   WavePlayer_0 = UE.Wave Player:Mono(Wave Asset=@ArrayRandomGet_0.Value, Pitch Shift=@FrequencyMultipliertoSemitone_0.Semitones, Play=@n1.On Next)
  ]
  links [{dst="@UE.Source.OneShot.OnFinished", src=WavePlayer_0.On Finished}, {dst="@UE.OutputFormat.Mono.Audio:0", src=WavePlayer_0.Out Mono}]
 }]
properties {
 assetImportData=&self:AssetImportData
 attenuationSettings=&Fort_Weapon_Reload_Att
 bOverrideConcurrency=true
 maxDistance=2250
 sampleRate=48000
 soundClassObject=&Fort_WeaponFoley
 volume=0.19
 concurrencyOverrides {bLimitToOwner=true, maxCount=1, resolutionRule=StopOldest}
}
