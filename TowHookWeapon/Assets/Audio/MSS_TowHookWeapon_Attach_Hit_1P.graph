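# One-shot stereo MetaSound Source for the tow hook weapon's first-person attach impact sound.
# On play it picks a random wave from the impact variations and rolls a random pitch shift, then waits briefly before starting the wave player.
# The player feeds the stereo output and signals finished when the wave ends.
# It uses first-person tow hook attenuation, with a concurrency override keeping one instance per owner by stopping the oldest.
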MetaSoundSource /TowHookWeapon/Assets/Audio/MSS_TowHookWeapon_Attach_Hit_1P
inputs [
 Sounds {type=WaveAsset:Array}
 UE.Source.OnPlay {type=Trigger}
]
interfaces [UE.OutputFormat.Stereo:1.0, UE.Source.OneShot:1.0, UE.Source:1.1]
outputs [
 "UE.OutputFormat.Stereo.Audio:0" {type=Audio}
 "UE.OutputFormat.Stereo.Audio:1" {type=Audio}
 UE.Source.OneShot.OnFinished {type=Trigger}
]
pages [Default {
  input_defaults {
   UE.Source.OnPlay=null
   Sounds [
    /TowHookWeapon/Assets/Audio/wav/TowHook_Proj_Impact_Attach_1P_01
    /TowHookWeapon/Assets/Audio/wav/TowHook_Proj_Impact_Attach_1P_02
    /TowHookWeapon/Assets/Audio/wav/TowHook_Proj_Impact_Attach_1P_03
   ]
  }
  nodes [
   ArrayRandomGet_0 = Array.Random Get:WaveAsset:Array(In Array=^Sounds, Next=^UE.Source.OnPlay)
   n1 = UE.RandomFloat(Next=@ArrayRandomGet_0.On Next) {also {defaults {Min=-1}}}
   TriggerDelay_0 = UE.Trigger Delay(In=@n1.On Next) {also {defaults {Delay Time=0.22}}}
   WavePlayer_0 = UE.Wave Player:Stereo(Wave Asset=@ArrayRandomGet_0.Value, Pitch Shift=UE.Reroute(Value=@n1.Value).Value, Play=@TriggerDelay_0.Out)
  ]
  links [
   {dst="@UE.Source.OneShot.OnFinished", src=WavePlayer_0.On Finished}
   {dst="@UE.OutputFormat.Stereo.Audio:0", src=WavePlayer_0.Out Left}
   {dst="@UE.OutputFormat.Stereo.Audio:1", src=WavePlayer_0.Out Right}
  ]
 }]
properties {
 assetImportData=&self:AssetImportData
 attenuationSettings=&/Att_TowHook_1P
 bOverrideConcurrency=true
 maxDistance=25200
 numChannels=2
 outputFormat=Stereo
 sampleRate=48000
 soundClassObject=&SFX
 volume=1.5
 concurrencyOverrides {bLimitToOwner=true, maxCount=1, resolutionRule=StopOldest}
}
