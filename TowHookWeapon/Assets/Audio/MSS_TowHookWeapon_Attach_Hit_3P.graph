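# One-shot mono MetaSound source for the third-person tow hook attach impact sound.
# On play, it picks a random wave from the Sounds array and rolls a random pitch shift, then starts the wave player after a short fixed delay.
# It uses tow hook attenuation, the impacts sound class and a concurrency override limiting overlapping instances.
# The Lead In array input is declared but left unconnected.
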MetaSoundSource /TowHookWeapon/Assets/Audio/MSS_TowHookWeapon_Attach_Hit_3P
inputs [
 Lead In {type=WaveAsset:Array}
 Sounds {type=WaveAsset:Array}
 UE.Source.OnPlay {type=Trigger}
]
interfaces [UE.OutputFormat.Mono:1.0, UE.Source.OneShot:1.0, UE.Source:1.1]
outputs [
 "UE.OutputFormat.Mono.Audio:0" {type=Audio}
 UE.Source.OneShot.OnFinished {type=Trigger}
]
pages [Default {
  input_defaults {
   UE.Source.OnPlay=null
   Lead In [null,null]
   Sounds [/TowHookWeapon/Assets/Audio/wav/TowHook_Proj_Impact_Attach_3P_01, /TowHookWeapon/Assets/Audio/wav/TowHook_Proj_Impact_Attach_3P_02]
  }
  nodes [
   ArrayRandomGet_0 = Array.Random Get:WaveAsset:Array(In Array=^Sounds, Next=^UE.Source.OnPlay)
   n1 = UE.RandomFloat(Next=@ArrayRandomGet_0.On Next) {also {defaults {Max=0.5, Min=-0.5}}}
   TriggerDelay_0 = UE.Trigger Delay(In=@n1.On Next) {also {defaults {Delay Time=0.35}}}
   WavePlayer_0 = UE.Wave Player:Mono(Wave Asset=@ArrayRandomGet_0.Value, Pitch Shift=@n1.Value, Play=@TriggerDelay_0.Out)
  ]
  links [{dst="@UE.Source.OneShot.OnFinished", src=WavePlayer_0.On Finished}, {dst="@UE.OutputFormat.Mono.Audio:0", src=WavePlayer_0.Out Mono}]
 }]
properties {
 assetImportData=&self:AssetImportData
 attenuationSettings=&/Att_TowHook
 bOverrideConcurrency=true
 maxDistance=4300
 sampleRate=48000
 soundClassObject=&Fort_Impacts
 concurrencyOverrides {maxCount=3}
}
