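# Third-person looping MetaSound for a player attached by the tow hook, voicing reel-in and reel-out movement during skidding.
# Skid-gated reel events pass speed checks and cooldowns, firing pitch-randomized start one-shots and loops from random offsets.
# Each direction releases the other's envelope; timed sustain envelopes also fade loops, whose mix drops with instigator distance.
# Randomly timed, high-passed wire wobbles repeat while skidding, unaffected by that distance scaling.
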MetaSoundSource /TowHookWeapon/Assets/Audio/MSS_TowHookWeapon_Attached_Player_Reeling_Loop_3P
inputs [
 Attack Time {type=Time}
 Cooldown Time {type=Time}
 DraggedIn.Loop {type=WaveAsset}
 DraggedIn.Starts {type=WaveAsset:Array}
 DraggedOut.Loop {type=WaveAsset}
 DraggedOut.Starts {type=WaveAsset:Array}
 Events.OnReelIn {type=Trigger}
 Events.OnReelOut {type=Trigger}
 Events.OnSkiddingEnd {type=Trigger}
 Events.OnSkiddingStart {type=Trigger}
 Input {type=Float}
 InstigatorDistance {type=Float}
 Loop Sustain Time {type=Time}
 Movement.Speed {type=Float}
 Movement.SpeedAzimuth {type=Float}
 Movement.SpeedElevation {type=Float}
 ReelIn.LoopVolume {type=Float}
 ReelOut.LoopVolume {type=Float}
 Release Time {type=Time}
 SpeedThreshold {type=Float}
 UE.Source.OnPlay {type=Trigger}
 Wobbles.Sounds {type=WaveAsset:Array}
]
interfaces [Fort.Movement:1.0, UE.OutputFormat.Mono:1.0, UE.Source:1.1]
outputs ["UE.OutputFormat.Mono.Audio:0" {type=Audio}]
pages [Default {
  input_defaults {
   Attack Time=0.3
   Cooldown Time=3
   DraggedIn.Loop=/TowHookWeapon/Assets/Audio/wav/TowHook_Mvmt_Loop_Dragged_In
   DraggedOut.Loop=/TowHookWeapon/Assets/Audio/wav/TowHook_Mvmt_Loop_Dragged_Out
   Events.OnReelIn=null
   Events.OnReelOut=null
   Events.OnSkiddingEnd=null
   Events.OnSkiddingStart=null
   Input=1896.0002
   InstigatorDistance=0
   Loop Sustain Time=6
   Movement.Speed=0
   Movement.SpeedAzimuth=0
   Movement.SpeedElevation=0
   ReelIn.LoopVolume=1
   ReelOut.LoopVolume=1
   Release Time=0.7
   SpeedThreshold=476
   UE.Source.OnPlay=null
   DraggedIn.Starts [/TowHookWeapon/Assets/Audio/wav/TowHook_Mvmt_Loop_Dragged_In_Start_01, /TowHookWeapon/Assets/Audio/wav/TowHook_Mvmt_Loop_Dragged_In_Start_02]
   DraggedOut.Starts [
    /TowHookWeapon/Assets/Audio/wav/TowHook_Mvmt_Loop_Dragged_Out_Start_01
    /TowHookWeapon/Assets/Audio/wav/TowHook_Mvmt_Loop_Dragged_Out_Start_02
   ]
   Wobbles.Sounds [
    /TowHookWeapon/Assets/Audio/wav/TowHook_Wire_Wobble_01
    /TowHookWeapon/Assets/Audio/wav/TowHook_Wire_Wobble_02
    /TowHookWeapon/Assets/Audio/wav/TowHook_Wire_Wobble_03
    /TowHookWeapon/Assets/Audio/wav/TowHook_Wire_Wobble_04
    /TowHookWeapon/Assets/Audio/wav/TowHook_Wire_Wobble_05
    /TowHookWeapon/Assets/Audio/wav/TowHook_Wire_Wobble_06
    /TowHookWeapon/Assets/Audio/wav/TowHook_Wire_Wobble_07
    /TowHookWeapon/Assets/Audio/wav/TowHook_Wire_Wobble_08
    /TowHookWeapon/Assets/Audio/wav/TowHook_Wire_Wobble_09
    /TowHookWeapon/Assets/Audio/wav/TowHook_Wire_Wobble_10
   ]
  }
  nodes [
   n1 = InitVariable.Trigger()
   n2 = VariableDeferredAccessor.Trigger(Variable=InitVariable.Trigger().Variable)
   TriggerAnyTriggerAny2_1 = "TriggerAny.Trigger Any (2)"(In 0=^Events.OnSkiddingStart, In 1=@n2.Value)
   TriggerControl_3 = UE.Trigger Control(Close=^Events.OnSkiddingEnd, Open=^Events.OnSkiddingStart, Trigger In=@TriggerAnyTriggerAny2_1.Out)
   n3 = UE.RandomTime(Next=@TriggerControl_3.Trigger Out) {also {defaults {Max=6, Min=2}}}
   n4 = UE.Pipe(In=@n3.On Next, Delay Time=@n3.Value)
   ArrayRandomGet_2 = Array.Random Get:WaveAsset:Array(
    In Array=^Wobbles.Sounds
    Next=VariableAccessor.Trigger(Variable=VariableMutator.Trigger(Variable=@n1.Variable, Value=@n4.Out).Variable).Value
   ) {also {defaults {Weights []}}}
   n5 = UE.RandomFloat(Next=@ArrayRandomGet_2.On Next) {also {defaults {Max=0.5, Min=-0.5}}}
   WavePlayer_0 = UE.Wave Player:Mono(Wave Asset=@ArrayRandomGet_2.Value, Play=@n5.On Next, Pitch Shift=@n5.Value)
   OnePoleHighPassFilter_0 = UE.One-Pole High Pass Filter:Audio(In=@WavePlayer_0.Out Mono) {also {defaults {Cutoff Frequency=800}}}
   n6 = VariableDeferredAccessor.Trigger(Variable=InitVariable.Trigger().Variable)
   ArrayRandomGet_1 = Array.Random Get:WaveAsset:Array(In Array=^DraggedOut.Starts, Next=@n6.Value) {also {defaults {Weights []}}}
   n7 = UE.RandomFloat(Next=@ArrayRandomGet_1.On Next) {also {defaults {Max=0.5, Min=-0.5}}}
   WavePlayer_2 = UE.Wave Player:Mono(Wave Asset=@ArrayRandomGet_1.Value, Play=@n7.On Next, Pitch Shift=@n7.Value)
   n8 = VariableDeferredAccessor.Trigger(Variable=InitVariable.Trigger().Variable)
   ArrayRandomGet_0 = Array.Random Get:WaveAsset:Array(In Array=^DraggedIn.Starts, Next=@n8.Value) {also {defaults {Weights []}}}
   n9 = UE.RandomFloat(Next=@ArrayRandomGet_0.On Next) {also {defaults {Max=0.5, Min=-0.5}}}
   WavePlayer_4 = UE.Wave Player:Mono(Wave Asset=@ArrayRandomGet_0.Value, Play=@n9.On Next, Pitch Shift=@n9.Value)
   AudioMixerAudioMixerMono3_0 = "AudioMixer.Audio Mixer (Mono, 3)"(In 2=@OnePoleHighPassFilter_0.Out, In 1=@WavePlayer_2.Out Mono, In 0=@WavePlayer_4.Out Mono)
   GetWaveDuration_1 = "UE.Get Wave Duration: "(Wave=^DraggedOut.Loop)
   n10 = InitVariable.Trigger()
   n11 = VariableDeferredAccessor.Trigger(Variable=@n10.Variable)
   n12 = VariableDeferredAccessor.Trigger(Variable=@n11.Variable)
   TriggerDelay_1 = UE.Trigger Delay(In=@n12.Value) {also {defaults {Delay Time=10}}}
   TriggerAccumulatorTriggerAccumulate2_0 = "TriggerAccumulator.Trigger Accumulate (2)"(In 0=@TriggerDelay_1.Out, In 1=VariableDeferredAccessor.Trigger(Variable=@n10.Variable).Value) {also {defaults {Auto Reset="<MetaSoundBuilderResult.SUCCEEDED: 0>"}}}
   TriggerControl_0 = UE.Trigger Control(Trigger In=^Events.OnReelOut, Close=^Events.OnSkiddingEnd, Open=^Events.OnSkiddingStart)
   TriggerControl_1 = UE.Trigger Control(Open=@TriggerAccumulatorTriggerAccumulate2_0.Out, Trigger In=@TriggerControl_0.Trigger Out, Close=@n11.Value)
   n13 = TriggerCompare.Clamp:Float(A=^Movement.Speed, B=^SpeedThreshold, Compare=@TriggerControl_1.Trigger Out) {also {defaults {Type=5}}}
   n14 = VariableAccessor.Trigger(Variable=VariableMutator.Trigger(Value=@n13.True, Variable=@n12.Variable).Variable)
   n15 = VariableAccessor.Trigger(Variable=@n14.Variable)
   n16 = UE.RandomTime(Max=@GetWaveDuration_1.Duration, Next=VariableAccessor.Trigger(Variable=@n15.Variable).Value)
   n17 = InitVariable.Trigger()
   n18 = InitVariable.Trigger()
   n19 = InitVariable.Trigger()
   n20 = VariableDeferredAccessor.Trigger(Variable=@n19.Variable)
   n21 = VariableDeferredAccessor.Trigger(Variable=@n20.Variable)
   TriggerDelay_2 = UE.Trigger Delay(Delay Time=^Cooldown Time, In=@n21.Value)
   TriggerAccumulatorTriggerAccumulate2_1 = "TriggerAccumulator.Trigger Accumulate (2)"(In 0=@TriggerDelay_2.Out, In 1=VariableDeferredAccessor.Trigger(Variable=@n19.Variable).Value) {also {defaults {Auto Reset="<MetaSoundBuilderResult.SUCCEEDED: 0>"}}}
   TriggerControl_2 = UE.Trigger Control(Trigger In=^Events.OnReelIn, Close=^Events.OnSkiddingEnd, Open=^Events.OnSkiddingStart)
   TriggerControl_4 = UE.Trigger Control(Open=@TriggerAccumulatorTriggerAccumulate2_1.Out, Trigger In=@TriggerControl_2.Trigger Out, Close=@n20.Value)
   n22 = TriggerCompare.Clamp:Float(A=^Movement.Speed, B=^SpeedThreshold, Compare=@TriggerControl_4.Trigger Out) {also {defaults {Type=5}}}
   n23 = VariableAccessor.Trigger(Variable=VariableMutator.Trigger(Value=@n22.True, Variable=@n21.Variable).Variable)
   n24 = VariableAccessor.Trigger(Variable=@n23.Variable)
   TriggerAnyTriggerAny2_4 = "TriggerAny.Trigger Any (2)"(In 0=^Events.OnSkiddingEnd, In 1=@n24.Value)
   ADSREnvelopeADSREnvelope_1 = ADSR Envelope.ADSR Envelope:Float(
    Attack Time=^Attack Time
    Release Time=^Release Time
    Trigger Release=VariableAccessor.Trigger(Variable=VariableMutator.Trigger(Variable=@n18.Variable, Value=@TriggerAnyTriggerAny2_4.Out).Variable).Value
    Trigger Attack=@n14.Value
   ) {also {defaults {Decay Time=0, Sustain Level=1}}}
   n25 = InitVariable.Trigger()
   n26 = VariableDeferredAccessor.Trigger(Variable=InitVariable.Trigger().Variable)
   TriggerAnyTriggerAny2_2 = "TriggerAny.Trigger Any (2)"(
    In 0=VariableAccessor.Trigger(Variable=VariableMutator.Trigger(Value=@ADSREnvelopeADSREnvelope_1.On Done, Variable=@n25.Variable).Variable).Value
    In 1=@n26.Value
   )
   WavePlayer_3 = UE.Wave Player:Mono(
    Wave Asset=^DraggedOut.Loop
    Play=@n16.On Next
    Start Time=@n16.Value
    Stop=VariableAccessor.Trigger(Variable=VariableMutator.Trigger(Variable=@n17.Variable, Value=@TriggerAnyTriggerAny2_2.Out).Variable).Value
   ) {also {defaults {Loop="<MetaSoundBuilderResult.SUCCEEDED: 0>"}}}
   TriggerDelay_3 = UE.Trigger Delay(Delay Time=^Loop Sustain Time, In=@WavePlayer_3.On Play)
   ADSREnvelopeADSREnvelope_0 = ADSR Envelope.ADSR Envelope:Float(Trigger Release=@TriggerDelay_3.Out, Trigger Attack=@WavePlayer_3.On Play) {also {defaults {Attack Time=0.1, Decay Time=0.5, Release Curve=0.5, Release Time=2, Sustain Level=1}}}
   n27 = UE.Reroute(Value=@WavePlayer_3.Out Mono)
   n28 = VariableMutator.Float(Value=@ADSREnvelopeADSREnvelope_1.Out Envelope, Variable=InitVariable.Float().Variable)
   GetWaveDuration_0 = "UE.Get Wave Duration: "(Wave=^DraggedIn.Loop)
   n29 = UE.RandomTime(Max=@GetWaveDuration_0.Duration, Next=VariableAccessor.Trigger(Variable=@n24.Variable).Value)
   n30 = InitVariable.Trigger()
   n31 = InitVariable.Trigger()
   TriggerAnyTriggerAny2_3 = "TriggerAny.Trigger Any (2)"(In 0=^Events.OnSkiddingEnd, In 1=@n15.Value)
   ADSREnvelopeADSREnvelope_2 = ADSR Envelope.ADSR Envelope:Float(
    Attack Time=^Attack Time
    Release Time=^Release Time
    Trigger Release=VariableAccessor.Trigger(Variable=VariableMutator.Trigger(Variable=@n31.Variable, Value=@TriggerAnyTriggerAny2_3.Out).Variable).Value
    Trigger Attack=@n23.Value
   ) {also {defaults {Decay Time=0, Sustain Level=1}}}
   n32 = InitVariable.Trigger()
   n33 = VariableDeferredAccessor.Trigger(Variable=InitVariable.Trigger().Variable)
   TriggerAnyTriggerAny2_0 = "TriggerAny.Trigger Any (2)"(
    In 0=VariableAccessor.Trigger(Variable=VariableMutator.Trigger(Value=@ADSREnvelopeADSREnvelope_2.On Done, Variable=@n32.Variable).Variable).Value
    In 1=@n33.Value
   )
   WavePlayer_1 = UE.Wave Player:Mono(
    Wave Asset=^DraggedIn.Loop
    Play=@n29.On Next
    Start Time=@n29.Value
    Stop=VariableAccessor.Trigger(Variable=VariableMutator.Trigger(Variable=@n30.Variable, Value=@TriggerAnyTriggerAny2_0.Out).Variable).Value
   ) {also {defaults {Loop="<MetaSoundBuilderResult.SUCCEEDED: 0>"}}}
   TriggerDelay_0 = UE.Trigger Delay(Delay Time=^Loop Sustain Time, In=@WavePlayer_1.On Play)
   ADSREnvelopeADSREnvelope_3 = ADSR Envelope.ADSR Envelope:Float(Trigger Release=@TriggerDelay_0.Out, Trigger Attack=@WavePlayer_1.On Play) {also {defaults {Attack Time=0.1, Decay Time=0.5, Release Curve=0.5, Release Time=2, Sustain Level=1}}}
   n34 = UE.Reroute(Value=@WavePlayer_1.Out Mono)
   n35 = VariableMutator.Float(Value=@ADSREnvelopeADSREnvelope_2.Out Envelope, Variable=InitVariable.Float().Variable)
   AudioMixerAudioMixerMono2_1 = "AudioMixer.Audio Mixer (Mono, 2)"(
    Gain 0=^ReelIn.LoopVolume
    Gain 1=^ReelOut.LoopVolume
    In 1=UE.Multiply:Audio by Float(
     PrimaryOperand=UE.Multiply:Audio by Float(AdditionalOperands=@ADSREnvelopeADSREnvelope_0.Out Envelope, PrimaryOperand=@n27.Value).Out
     AdditionalOperands=VariableAccessor.Float(Variable=@n28.Variable).Value
    ).Out
    In 0=UE.Multiply:Audio by Float(
     PrimaryOperand=UE.Multiply:Audio by Float(AdditionalOperands=@ADSREnvelopeADSREnvelope_3.Out Envelope, PrimaryOperand=@n34.Value).Out
     AdditionalOperands=VariableAccessor.Float(Variable=@n35.Variable).Value
    ).Out
   )
   AudioMixerAudioMixerMono2_0 = "AudioMixer.Audio Mixer (Mono, 2)"(
    In 1=@AudioMixerAudioMixerMono3_0.Out
    In 0=UE.Multiply:Audio by Float(
     PrimaryOperand=@AudioMixerAudioMixerMono2_1.Out
     AdditionalOperands=MapRange.MapRange:Float(In=^InstigatorDistance) {also {defaults {In Range B=1000, Out Range A=1, Out Range B=0}}}.Out Value
    ).Out
   ) {also {defaults {Gain 0=1.5}}}
   VariableMutator.Trigger(Value=@n13.True, Variable=@n6.Variable)
   VariableMutator.Trigger(Value=@ADSREnvelopeADSREnvelope_0.On Done, Variable=@n26.Variable)
   VariableMutator.Trigger(Variable=@n2.Variable, Value=@WavePlayer_0.On Finished)
   VariableMutator.Trigger(Value=@n22.True, Variable=@n8.Variable)
   VariableMutator.Trigger(Value=@ADSREnvelopeADSREnvelope_3.On Done, Variable=@n33.Variable)
  ]
  links [{dst="@UE.OutputFormat.Mono.Audio:0", src=AudioMixerAudioMixerMono2_0.Out}]
 }]
properties {
 assetImportData=&self:AssetImportData
 attenuationSettings=&/Att_TowHook_Reeling_Loop_3P
 bLooping=true
 bOverrideConcurrency=true
 duration=10000
 maxDistance=2400
 sampleRate=48000
 soundClassObject=&Fort_Gadgets
 volume=0.4
 concurrencyOverrides {maxCount=2}
}
