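# Looping mono MetaSound source for TowHookWeapon's attached skidding sound, a NoGrain variant that layers surface audio with a player-dragged jingle.
# A crossfade patch swaps surface sounds on surface-change triggers, while a check on the new wave's duration gates the looping jingle through an interpolated gain.
# The mix is scaled by smoothed movement speed, rising faster than it falls, and by character affiliation, louder for enemies.
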MetaSoundSource /TowHookWeapon/Assets/Audio/MSS_TowHookWeapon_Attached_Skidding_Loop_NoGrain
inputs [
 Character.Affiliation {type=Enum:FortAudioAffiliation}
 Character.Health {type=Float}
 Character.UniqueID {type=Int32}
 Events.OnSurfaceChanged {type=Trigger}
 Input {type=Float}
 Jingles {type=WaveAsset}
 Movement.Speed {type=Float}
 Movement.SpeedAzimuth {type=Float}
 Movement.SpeedElevation {type=Float}
 Surfaces.CrossfadeTime {type=Time}
 Surfaces.FadeOut {type=Trigger}
 Surfaces.Pitch {type=Float}
 Surfaces.Sound {type=WaveAsset}
 Surfaces.Volume {type=Float}
 UE.Source.OnPlay {type=Trigger}
 Wobbles {type=WaveAsset:Array}
]
interfaces [Fort.Character:1.0, Fort.Movement:1.0, UE.OutputFormat.Mono:1.0, UE.Source:1.1]
outputs ["UE.OutputFormat.Mono.Audio:0" {type=Audio}]
pages [Default {
  input_defaults {
   Character.Affiliation=0
   Character.Health=0
   Character.UniqueID=0
   Events.OnSurfaceChanged=null
   Input=216.0012
   Jingles=/TowHookWeapon/Assets/Audio/wav/TowHook_Mvmt_Loop_Dragged_Player_Jingle
   Movement.Speed=0
   Movement.SpeedAzimuth=0
   Movement.SpeedElevation=0
   Surfaces.CrossfadeTime=0.3
   Surfaces.FadeOut=null
   Surfaces.Pitch=0
   Surfaces.Sound=null
   Surfaces.Volume=1
   UE.Source.OnPlay=null
   Wobbles [
    /TowHookWeapon/Assets/Audio/wav/TowHook_Wire_Wobble_01
    /TowHookWeapon/Assets/Audio/wav/TowHook_Wire_Wobble_02
    /TowHookWeapon/Assets/Audio/wav/TowHook_Wire_Wobble_03
    /TowHookWeapon/Assets/Audio/wav/TowHook_Wire_Wobble_04
    /TowHookWeapon/Assets/Audio/wav/TowHook_Wire_Wobble_05
    /TowHookWeapon/Assets/Audio/wav/TowHook_Wire_Wobble_06
    /TowHookWeapon/Assets/Audio/wav/TowHook_Wire_Wobble_07
    /TowHookWeapon/Assets/Audio/wav/TowHook_Wire_Wobble_08
    /TowHookWeapon/Assets/Audio/wav/TowHook_Wire_Wobble_09
    /TowHookWeapon/Assets/Audio/wav/TowHook_Wire_Wobble_10
   ]
  }
  nodes [
   InitVariable.Trigger()
   InitVariable.Trigger()
   GetWaveDuration_0 = "UE.Get Wave Duration: "(Wave=^Surfaces.Sound)
   n1 = TriggerCompare.Clamp:Float(Compare=^Events.OnSurfaceChanged, A=UE.ConversionTimeToFloat(In=@GetWaveDuration_0.Duration).Out) {also {defaults {B=0}}}
   TriggerRouteTriggerRouteBool2_0 = "TriggerRoute.Trigger Route (Bool, 2):Bool"(Set 1=@n1.False, Set 0=@n1.True) {also {defaults {Value 1="<MetaSoundBuilderResult.SUCCEEDED: 0>"}}}
   n2 = Convert.Float:Bool(Bool=@TriggerRouteTriggerRouteBool2_0.Value)
   WavePlayer_0 = UE.Wave Player:Mono(Wave Asset=^Jingles, Play=^UE.Source.OnPlay) {also {defaults {Loop="<MetaSoundBuilderResult.SUCCEEDED: 0>"}}}
   n3 = &/MSP_TowHookWeapon_SurfaceCrossfade(
    PlayNextSound=^Events.OnSurfaceChanged
    CrossfadeTime=^Surfaces.CrossfadeTime
    FadeOutAllSounds=^Surfaces.FadeOut
    PitchShift=^Surfaces.Pitch
    NextSoundToPlay=^Surfaces.Sound
   )
   AudioMixerAudioMixerMono4_0 = "AudioMixer.Audio Mixer (Mono, 4)"(
    In 2=UE.Multiply:Audio by Float(AdditionalOperands=UE.InterpTo:Audio(Target=@n2.Float).Value, PrimaryOperand=UE.Reroute(Value=@WavePlayer_0.Out Mono).Value).Out
    In 0=UE.Reroute(Value=UE.Multiply:Audio by Float(AdditionalOperands=^Surfaces.Volume, PrimaryOperand=@n3.Out).Out).Value
   ) {also {defaults {Gain 0=0.6, Gain 1=0.6, Gain 2=0.1, Gain 3=1.1}}}
   n4 = MapRange.MapRange:Float(In=^Movement.Speed) {also {defaults {In Range A=50, In Range B=3000}}}
   Multiply_1 = UE.Multiply:Audio by Float(
    AdditionalOperands=&MSP_AffiliationMixFloat(Affiliation=^Character.Affiliation) {also {defaults {Enemy=1.2, Non-Enemy=0.5}}}.Out
    PrimaryOperand=UE.Multiply:Audio by Float(
     PrimaryOperand=@AudioMixerAudioMixerMono4_0.Out
     AdditionalOperands=&InterpSpeed(FloatInput=@n4.Out Value) {also {defaults {FallingInterpSpeed=0.5, RisingInterpSpeed=2}}}.FloatOutput
    ).Out
   )
  ]
  links [{dst="@UE.OutputFormat.Mono.Audio:0", src=Multiply_1.Out}]
 }]
properties {
 assetImportData=&self:AssetImportData
 attenuationSettings=&/Att_TowHook_Stress_Loop
 bLooping=true
 bOverrideConcurrency=true
 duration=10000
 maxDistance=3960
 sampleRate=48000
 soundClassObject=&Fort_Gadgets
 volume=2
 concurrencyOverrides {maxCount=3, retriggerTime=0.2}
}
