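# Looping mono MetaSound for the TowHookWeapon's vehicle stress sound while attached, driven by a Speed input.
# Speed threshold crossings trigger an ADSR attack and release, starting a wave from Sounds at a random offset and stopping it once the envelope finishes.
# Speed above the threshold maps to an interpolated volume.
# A fatigue envelope ducks the loop after a delay past the threshold, then recovers on release.
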MetaSoundSource /TowHookWeapon/Assets/Audio/MSS_TowHookWeapon_Attached_Vehicle_Loop
inputs [
 Attack Time {type=Time}
 Fatigue.DuckAmount {type=Float}
 Fatigue.FadeTime {type=Time}
 Fatigue.ResetFadeTime {type=Time}
 Fatigue.Threshold {type=Time}
 Min Loop Volume {type=Float}
 Release Time {type=Time}
 Sounds {type=WaveAsset:Array}
 Speed {type=Float}
 SpeedThreshold {type=Float}
 UE.Source.OnPlay {type=Trigger}
]
interfaces [UE.OutputFormat.Mono:1.0, UE.Source:1.1]
outputs ["UE.OutputFormat.Mono.Audio:0" {type=Audio}]
pages [Default {
  input_defaults {
   Attack Time=1
   Fatigue.DuckAmount=0.4
   Fatigue.FadeTime=5
   Fatigue.ResetFadeTime=5
   Fatigue.Threshold=2
   Min Loop Volume=0.125
   Release Time=2
   Speed=0
   SpeedThreshold=500
   UE.Source.OnPlay=null
   Sounds [/TowHookWeapon/Assets/Audio/wav/TowHook_Mvmt_Loop_VehicleStress]
  }
  nodes [
   n1 = VariableAccessor.Trigger(Variable=VariableMutator.Trigger(Variable=InitVariable.Trigger().Variable, Value=UE.TriggerOnThreshold:Float(In=^Speed, Threshold=^SpeedThreshold).Out).Variable)
   n2 = VariableAccessor.Trigger(Variable=@n1.Variable)
   TriggerDelay_0 = UE.Trigger Delay(Delay Time=^Fatigue.Threshold, In=VariableAccessor.Trigger(Variable=@n2.Variable).Value)
   n3 = VariableAccessor.Trigger(Variable=VariableMutator.Trigger(
     Variable=InitVariable.Trigger().Variable
     Value=UE.TriggerOnThreshold:Float(In=^Speed, Threshold=^SpeedThreshold) {also {defaults {Type=1}}}.Out
    ).Variable)
   ADSREnvelopeADSREnvelope_1 = ADSR Envelope.ADSR Envelope:Float(
    Attack Time=^Fatigue.FadeTime
    Release Time=^Fatigue.ResetFadeTime
    Trigger Attack=@TriggerDelay_0.Out
    Trigger Release=VariableAccessor.Trigger(Variable=@n3.Variable).Value
   ) {also {defaults {Decay Time=0, Sustain Level=1}}}
   n4 = UE.Multiply:Float(PrimaryOperand=^Fatigue.DuckAmount) {also {defaults {AdditionalOperands=-1}}}
   n5 = MapRange.MapRange:Float(In=UE.Subtract:Float(PrimaryOperand=^Speed, AdditionalOperands=^SpeedThreshold).Out) {also {defaults {In Range B=3000}}}
   ADSREnvelopeADSREnvelope_0 = ADSR Envelope.ADSR Envelope:Float(Attack Time=^Attack Time, Release Time=^Release Time, Trigger Attack=@n2.Value, Trigger Release=@n3.Value) {also {defaults {Decay Time=4, Release Curve=1.5, Sustain Level=0.7}}}
   ArrayRandomGet_0 = Array.Random Get:WaveAsset:Array(In Array=^Sounds, Next=@n1.Value) {also {defaults {Weights []}}}
   GetWaveDuration_0 = "UE.Get Wave Duration: "(Wave=@ArrayRandomGet_0.Value)
   n6 = UE.RandomTime(Next=@ArrayRandomGet_0.On Next, Max=@GetWaveDuration_0.Duration)
   n7 = InitVariable.Trigger()
   WavePlayer_0 = UE.Wave Player:Mono(
    Wave Asset=@ArrayRandomGet_0.Value
    Play=@n6.On Next
    Start Time=@n6.Value
    Stop=VariableAccessor.Trigger(Variable=VariableMutator.Trigger(Value=@ADSREnvelopeADSREnvelope_0.On Done, Variable=@n7.Variable).Variable).Value
   ) {also {defaults {Loop="<MetaSoundBuilderResult.SUCCEEDED: 0>"}}}
   Multiply_1 = UE.Multiply:Audio by Float(
    AdditionalOperands=UE.Add:Float(PrimaryOperand=UE.Multiply:Float(AdditionalOperands=@ADSREnvelopeADSREnvelope_1.Out Envelope, PrimaryOperand=@n4.Out).Out) {also {defaults {AdditionalOperands=1}}}.Out
    PrimaryOperand=UE.Multiply:Audio by Float(
     AdditionalOperands=&InterpSpeed(FloatInput=@n5.Out Value) {also {defaults {FallingInterpSpeed=1, RisingInterpSpeed=1}}}.FloatOutput
     PrimaryOperand=UE.Multiply:Audio by Float(AdditionalOperands=@ADSREnvelopeADSREnvelope_0.Out Envelope, PrimaryOperand=@WavePlayer_0.Out Mono).Out
    ).Out
   )
  ]
  links [{dst="@UE.OutputFormat.Mono.Audio:0", src=Multiply_1.Out}]
 }]
properties {
 assetImportData=&self:AssetImportData
 attenuationSettings=&/Att_TowHook_Stress_Loop
 bDebug=true
 bLooping=true
 bOverrideConcurrency=true
 duration=10000
 maxDistance=3960
 sampleRate=48000
 soundClassObject=/Game/Sounds/FortSoundClassesAndModes/Fort_06_Player_Attacks_Abilities
 volume=1.2
 concurrencyOverrides {bLimitToOwner=true, maxCount=2, retriggerTime=0.2}
}
