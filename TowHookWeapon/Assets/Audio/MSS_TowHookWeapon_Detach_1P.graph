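# One-shot stereo MetaSound for the tow hook weapon's first-person detach sound.
# On play, it starts a detach wave from Sounds, plus a randomly chosen wire-snap whip layer after a short random delay.
# Each layer gets random pitch shift before mixing, with the main louder; only the main player signals finish, and the Wire Snap input is unused.
# Concurrency allows one instance per owner, stopping the oldest.
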MetaSoundSource /TowHookWeapon/Assets/Audio/MSS_TowHookWeapon_Detach_1P
inputs [
 Extra Whip Layer {type=WaveAsset:Array}
 Sounds {type=WaveAsset:Array}
 UE.Source.OnPlay {type=Trigger}
 Wire Snap {type=WaveAsset}
]
interfaces [UE.OutputFormat.Stereo:1.0, UE.Source.OneShot:1.0, UE.Source:1.1]
outputs [
 "UE.OutputFormat.Stereo.Audio:0" {type=Audio}
 "UE.OutputFormat.Stereo.Audio:1" {type=Audio}
 UE.Source.OneShot.OnFinished {type=Trigger}
]
pages [Default {
  input_defaults {
   UE.Source.OnPlay=null
   Wire Snap=null
   Extra Whip Layer [
    /TowHookWeapon/Assets/Audio/wav/TowHook_Wire_Snap_Detatch_Layer_01
    /TowHookWeapon/Assets/Audio/wav/TowHook_Wire_Snap_Detatch_Layer_02
    /TowHookWeapon/Assets/Audio/wav/TowHook_Wire_Snap_Detatch_Layer_03
    /TowHookWeapon/Assets/Audio/wav/TowHook_Wire_Snap_Detatch_Layer_04
   ]
   Sounds [/TowHookWeapon/Assets/Audio/wav/TowHook_Proj_Detatch_1P]
  }
  nodes [
   ArrayRandomGet_0 = Array.Random Get:WaveAsset:Array(In Array=^Sounds, Next=^UE.Source.OnPlay)
   n1 = UE.RandomFloat(Next=@ArrayRandomGet_0.On Next) {also {defaults {Max=0.5, Min=-0.5}}}
   WavePlayer_0 = UE.Wave Player:Stereo(Wave Asset=@ArrayRandomGet_0.Value, Play=@n1.On Next, Pitch Shift=@n1.Value)
   n2 = UE.RandomTime(Next=^UE.Source.OnPlay) {also {defaults {Max=0.4, Min=0.2}}}
   TriggerDelay_0 = UE.Trigger Delay(In=@n2.On Next, Delay Time=@n2.Value)
   ArrayRandomGet_1 = Array.Random Get:WaveAsset:Array(In Array=^Extra Whip Layer, Next=@TriggerDelay_0.Out)
   n3 = UE.RandomFloat(Next=@ArrayRandomGet_1.On Next) {also {defaults {Max=0.5, Min=-0.5}}}
   WavePlayer_1 = UE.Wave Player:Stereo(Wave Asset=@ArrayRandomGet_1.Value, Play=@n3.On Next, Pitch Shift=@n3.Value)
   AudioMixerAudioMixerStereo2_0 = "AudioMixer.Audio Mixer (Stereo, 2)"(In 0 L=@WavePlayer_0.Out Left, In 0 R=@WavePlayer_0.Out Right, In 1 L=@WavePlayer_1.Out Left, In 1 R=@WavePlayer_1.Out Right) {also {defaults {Gain 0=1.5, Gain 1=0.8}}}
  ]
  links [
   {dst="@UE.OutputFormat.Stereo.Audio:0", src=AudioMixerAudioMixerStereo2_0.Out L}
   {dst="@UE.OutputFormat.Stereo.Audio:1", src=AudioMixerAudioMixerStereo2_0.Out R}
   {dst="@UE.Source.OneShot.OnFinished", src=WavePlayer_0.On Finished}
  ]
 }]
properties {
 assetImportData=&self:AssetImportData
 attenuationSettings=&/Att_TowHook_1P
 bOverrideConcurrency=true
 maxDistance=25200
 numChannels=2
 outputFormat=Stereo
 sampleRate=48000
 soundClassObject=&SFX
 volume=1.45
 concurrencyOverrides {bLimitToOwner=true, maxCount=1, resolutionRule=StopOldest}
}
