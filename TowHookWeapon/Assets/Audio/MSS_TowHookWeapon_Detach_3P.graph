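# One-shot mono MetaSound for the third-person detach sound of the tow hook weapon's projectile.
# On play it picks from the Sounds array, currently one wave, applies a slight random pitch shift and finishes when the wave ends.
# A Wire Snap array input is declared but empty and unconnected.
# It uses its own detach attenuation asset, a concurrency cap and the Fortnite player abilities sound class.
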MetaSoundSource /TowHookWeapon/Assets/Audio/MSS_TowHookWeapon_Detach_3P
inputs [
 Sounds {type=WaveAsset:Array}
 UE.Source.OnPlay {type=Trigger}
 Wire Snap {type=WaveAsset:Array}
]
interfaces [UE.OutputFormat.Mono:1.0, UE.Source.OneShot:1.0, UE.Source:1.1]
outputs [
 "UE.OutputFormat.Mono.Audio:0" {type=Audio}
 UE.Source.OneShot.OnFinished {type=Trigger}
]
pages [Default {
  input_defaults {UE.Source.OnPlay=null, Sounds [/TowHookWeapon/Assets/Audio/wav/TowHook_Proj_Detatch_3P], Wire Snap [null,null,null,null]}
  nodes [
   ArrayRandomGet_0 = Array.Random Get:WaveAsset:Array(In Array=^Sounds, Next=^UE.Source.OnPlay)
   n1 = UE.RandomFloat(Next=@ArrayRandomGet_0.On Next) {also {defaults {Max=0.5, Min=-0.5}}}
   WavePlayer_0 = UE.Wave Player:Mono(Wave Asset=@ArrayRandomGet_0.Value, Play=@n1.On Next, Pitch Shift=@n1.Value)
  ]
  links [{dst="@UE.Source.OneShot.OnFinished", src=WavePlayer_0.On Finished}, {dst="@UE.OutputFormat.Mono.Audio:0", src=WavePlayer_0.Out Mono}]
 }]
properties {
 assetImportData=&self:AssetImportData
 attenuationSettings=&/Att_TowHook_Projectile_Detach_3P
 bOverrideConcurrency=true
 maxDistance=3000
 sampleRate=48000
 soundClassObject=/Game/Sounds/FortSoundClassesAndModes/Fort_06_Player_Attacks_Abilities
 concurrencyOverrides {maxCount=3}
}
