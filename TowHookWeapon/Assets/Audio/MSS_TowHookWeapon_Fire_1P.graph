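# 5.1 one-shot MetaSound for the first-person Tow Hook weapon fire, limited to one instance with stop-oldest concurrency.
# Fire plays a random pitch-shifted shot, a delayed random wire layer, and a quieter last-shot layer gated by bIsLastShot.
# An ADSR envelope fades and stops the wire on StopWire.
# Layers mix into front left/right; only the main shot feeds other channels and signals OnFinished.
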MetaSoundSource /TowHookWeapon/Assets/Audio/MSS_TowHookWeapon_Fire_1P
inputs [
 bIsLastShot {type=Bool}
 Events.Fire {type=Trigger}
 Events.StopWire {type=Trigger}
 LastShots {type=WaveAsset:Array}
 Sounds {type=WaveAsset:Array}
 UE.Source.OnPlay {type=Trigger}
 Wires {type=WaveAsset:Array}
]
interfaces [UE.OutputFormat.5dot1:1.0, UE.Source.OneShot:1.0, UE.Source:1.1]
outputs [
 "UE.OutputFormat.5dot1.Audio:0" {type=Audio}
 "UE.OutputFormat.5dot1.Audio:1" {type=Audio}
 "UE.OutputFormat.5dot1.Audio:2" {type=Audio}
 "UE.OutputFormat.5dot1.Audio:3" {type=Audio}
 "UE.OutputFormat.5dot1.Audio:4" {type=Audio}
 "UE.OutputFormat.5dot1.Audio:5" {type=Audio}
 UE.Source.OneShot.OnFinished {type=Trigger}
]
pages [Default {
  input_defaults {
   Events.Fire=null
   Events.StopWire=null
   UE.Source.OnPlay=null
   bIsLastShot=null
   LastShots [/TowHookWeapon/Assets/Audio/wav/TowHook_Weapon_Shoot_LastShot]
   Sounds [
    /TowHookWeapon/Assets/Audio/wav/TowHook_Weapon_Shoot_1P_01
    /TowHookWeapon/Assets/Audio/wav/TowHook_Weapon_Shoot_1P_02
    /TowHookWeapon/Assets/Audio/wav/TowHook_Weapon_Shoot_1P_03
   ]
   Wires [
    /TowHookWeapon/Assets/Audio/wav/TowHook_Weapon_Shoot_1P_Wire_01
    /TowHookWeapon/Assets/Audio/wav/TowHook_Weapon_Shoot_1P_Wire_02
    /TowHookWeapon/Assets/Audio/wav/TowHook_Weapon_Shoot_1P_Wire_03
   ]
  }
  nodes [
   ADSREnvelopeADSREnvelope_0 = ADSR Envelope.ADSR Envelope:Float(Trigger Attack=^Events.Fire, Trigger Release=^Events.StopWire) {also {defaults {Attack Time=0, Decay Time=0, Release Time=0.2, Sustain Level=1}}}
   n1 = UE.RandomTime(Next=^Events.Fire) {also {defaults {Max=0.4, Min=0.4}}}
   TriggerDelay_0 = UE.Trigger Delay(In=@n1.On Next, Delay Time=@n1.Value)
   ArrayRandomGet_0 = Array.Random Get:WaveAsset:Array(In Array=^Wires, Next=@TriggerDelay_0.Out)
   n2 = UE.RandomFloat(Next=@ArrayRandomGet_0.On Next) {also {defaults {Max=0.5}}}
   WavePlayer_1 = UE.Wave Player:Stereo(Stop=@ADSREnvelopeADSREnvelope_0.On Done, Wave Asset=@ArrayRandomGet_0.Value, Play=@n2.On Next, Pitch Shift=@n2.Value)
   ArrayRandomGet_1 = Array.Random Get:WaveAsset:Array(Next=^Events.Fire, In Array=^Sounds)
   n3 = UE.RandomFloat(Next=@ArrayRandomGet_1.On Next) {also {defaults {Max=0.5, Min=-0.5}}}
   WavePlayer_0 = UE.Wave Player:5dot1(Wave Asset=@ArrayRandomGet_1.Value, Play=@n3.On Next, Pitch Shift=@n3.Value)
   ArrayRandomGet_2 = Array.Random Get:WaveAsset:Array(
    In Array=^LastShots
    Next=TriggerCompare.Clamp:Bool(Compare=^Events.Fire, A=^bIsLastShot) {also {defaults {B="<MetaSoundBuilderResult.SUCCEEDED: 0>"}}}.True
   )
   n4 = UE.RandomFloat(Next=@ArrayRandomGet_2.On Next) {also {defaults {Max=0.5, Min=-0.5}}}
   WavePlayer_2 = UE.Wave Player:Stereo(Wave Asset=@ArrayRandomGet_2.Value, Play=@n4.On Next, Pitch Shift=@n4.Value)
   AudioMixerAudioMixerStereo3_0 = "AudioMixer.Audio Mixer (Stereo, 3)"(
    In 1 L=UE.Multiply:Audio by Float(AdditionalOperands=@ADSREnvelopeADSREnvelope_0.Out Envelope, PrimaryOperand=@WavePlayer_1.Out Left).Out
    In 1 R=UE.Multiply:Audio by Float(AdditionalOperands=@ADSREnvelopeADSREnvelope_0.Out Envelope, PrimaryOperand=@WavePlayer_1.Out Right).Out
    In 0 L=@WavePlayer_0.Out Front Left
    In 0 R=@WavePlayer_0.Out Front Right
    In 2 L=@WavePlayer_2.Out Left
    In 2 R=@WavePlayer_2.Out Right
   ) {also {defaults {Gain 2=0.5}}}
  ]
  links [
   {dst="@UE.OutputFormat.5dot1.Audio:0", src=AudioMixerAudioMixerStereo3_0.Out L}
   {dst="@UE.OutputFormat.5dot1.Audio:1", src=AudioMixerAudioMixerStereo3_0.Out R}
   {dst="@UE.Source.OneShot.OnFinished", src=WavePlayer_0.On Finished}
   {dst="@UE.OutputFormat.5dot1.Audio:2", src=WavePlayer_0.Out Front Center}
   {dst="@UE.OutputFormat.5dot1.Audio:3", src=WavePlayer_0.Out Low Frequency}
   {dst="@UE.OutputFormat.5dot1.Audio:4", src=WavePlayer_0.Out Side Left}
   {dst="@UE.OutputFormat.5dot1.Audio:5", src=WavePlayer_0.Out Side Right}
  ]
 }]
properties {
 assetImportData=&self:AssetImportData
 bDebug=true
 bOverrideConcurrency=true
 numChannels=6
 outputFormat=FiveDotOne
 sampleRate=48000
 soundClassObject=&Fort_WeaponRanged_Fire_1P_PSM
 volume=0.7
 concurrencyOverrides {maxCount=1, resolutionRule=StopOldest}
}
