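# Mono one-shot MetaSound for the tow hook weapon's third-person fire sound, layered by listener distance.
# On play, close, medium and distant wave players each get a random clip and random pitch shift.
# Attenuation distance drives mixer gains: close fades out over near range, then medium crossfades into distant, each scaled by a volume input.
# A trigger accumulator signals finished once all three players end; concurrency override stops the quietest.
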MetaSoundSource /TowHookWeapon/Assets/Audio/MSS_TowHookWeapon_Fire_3P
inputs [
 Close.Volume {type=Float}
 Closest {type=WaveAsset:Array}
 Distant.Volume {type=Float}
 Medium.Volume {type=Float}
 Sounds.Distant {type=WaveAsset:Array}
 Sounds.Medium {type=WaveAsset:Array}
 UE.Attenuation.Distance {type=Float}
 UE.Source.OnPlay {type=Trigger}
]
interfaces [UE.Attenuation:1.0, UE.OutputFormat.Mono:1.0, UE.Source.OneShot:1.0, UE.Source:1.1]
outputs [
 "UE.OutputFormat.Mono.Audio:0" {type=Audio}
 UE.Source.OneShot.OnFinished {type=Trigger}
]
pages [Default {
  input_defaults {
   Close.Volume=3
   Distant.Volume=0.3
   Medium.Volume=1.4
   UE.Attenuation.Distance=0
   UE.Source.OnPlay=null
   Closest [
    /TowHookWeapon/Assets/Audio/wav/TowHook_Shoot_3P_Close_01
    /TowHookWeapon/Assets/Audio/wav/TowHook_Shoot_3P_Close_02
    /TowHookWeapon/Assets/Audio/wav/TowHook_Shoot_3P_Close_03
   ]
   Sounds.Distant [
    /TowHookWeapon/Assets/Audio/wav/TowHook_Shoot_3P_Dist_01
    /TowHookWeapon/Assets/Audio/wav/TowHook_Shoot_3P_Dist_02
    /TowHookWeapon/Assets/Audio/wav/TowHook_Shoot_3P_Dist_03
   ]
   Sounds.Medium [
    /TowHookWeapon/Assets/Audio/wav/TowHook_Shoot_3P_01
    /TowHookWeapon/Assets/Audio/wav/TowHook_Shoot_3P_02
    /TowHookWeapon/Assets/Audio/wav/TowHook_Shoot_3P_03
   ]
  }
  nodes [
   n1 = MapRange.MapRange:Float(In=^UE.Attenuation.Distance) {also {defaults {In Range A=3600, In Range B=10000}}}
   n2 = UE.Subtract:Float(AdditionalOperands=MapRange.MapRange:Float(In=^UE.Attenuation.Distance) {also {defaults {In Range B=3600}}}.Out Value) {also {defaults {PrimaryOperand=1}}}
   n3 = UE.Subtract:Float(AdditionalOperands=@n1.Out Value) {also {defaults {PrimaryOperand=1}}}
   n4 = InitVariable.Audio()
   ArrayRandomGet_0 = Array.Random Get:WaveAsset:Array(In Array=^Sounds.Distant, Next=^UE.Source.OnPlay)
   n5 = UE.RandomFloat(Next=@ArrayRandomGet_0.On Next) {also {defaults {Max=0.5, Min=-0.5}}}
   WavePlayer_0 = UE.Wave Player:Mono(Wave Asset=@ArrayRandomGet_0.Value, Play=@n5.On Next, Pitch Shift=@n5.Value)
   n6 = InitVariable.Audio()
   ArrayRandomGet_1 = Array.Random Get:WaveAsset:Array(In Array=^Sounds.Medium, Next=^UE.Source.OnPlay) {also {defaults {Weights []}}}
   n7 = UE.RandomFloat(Next=@ArrayRandomGet_1.On Next) {also {defaults {Max=0.5, Min=-0.5}}}
   WavePlayer_2 = UE.Wave Player:Mono(Wave Asset=@ArrayRandomGet_1.Value, Play=@n7.On Next, Pitch Shift=@n7.Value)
   n8 = InitVariable.Audio()
   ArrayRandomGet_2 = Array.Random Get:WaveAsset:Array(In Array=^Closest, Next=^UE.Source.OnPlay)
   n9 = UE.RandomFloat(Next=@ArrayRandomGet_2.On Next) {also {defaults {Max=0.5, Min=-0.5}}}
   WavePlayer_1 = UE.Wave Player:Mono(Wave Asset=@ArrayRandomGet_2.Value, Play=@n9.On Next, Pitch Shift=@n9.Value)
   AudioMixerAudioMixerMono3_0 = "AudioMixer.Audio Mixer (Mono, 3)"(
    Gain 2=UE.Multiply:Float(AdditionalOperands=^Distant.Volume, PrimaryOperand=UE.Reroute(Value=@n1.Out Value).Value).Out
    Gain 0=UE.Reroute(Value=UE.Multiply:Float(AdditionalOperands=^Close.Volume, PrimaryOperand=@n2.Out).Out).Value
    Gain 1=UE.Reroute(Value=UE.Multiply:Float(AdditionalOperands=^Medium.Volume, PrimaryOperand=@n3.Out).Out).Value
    In 2=VariableAccessor.Audio(Variable=VariableMutator.Audio(Variable=@n4.Variable, Value=@WavePlayer_0.Out Mono).Variable).Value
    In 1=VariableAccessor.Audio(Variable=VariableMutator.Audio(Variable=@n6.Variable, Value=@WavePlayer_2.Out Mono).Variable).Value
    In 0=VariableAccessor.Audio(Variable=VariableMutator.Audio(Variable=@n8.Variable, Value=@WavePlayer_1.Out Mono).Variable).Value
   )
   TriggerAccumulatorTriggerAccumulate3_0 = "TriggerAccumulator.Trigger Accumulate (3)"(In 2=@WavePlayer_0.On Finished, In 0=@WavePlayer_1.On Finished, In 1=@WavePlayer_2.On Finished)
  ]
  links [
   {dst="@UE.OutputFormat.Mono.Audio:0", src=AudioMixerAudioMixerMono3_0.Out}
   {dst="@UE.Source.OneShot.OnFinished", src=TriggerAccumulatorTriggerAccumulate3_0.Out}
  ]
 }]
properties {
 assetImportData=&self:AssetImportData
 attenuationSettings=&Fort_Weapon_Projectile_NonAtt_Athena_Att
 bDebug=true
 bOverrideConcurrency=true
 maxDistance=27500
 sampleRate=48000
 soundClassObject=&Fort_WeaponRanged_Fire_3P_PSM
 volume=1.1
 concurrencyOverrides {maxCount=4, resolutionRule=StopQuietest}
}
