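# One-shot mono MetaSound source for the tow hook weapon's projectile impact sound, in its generic variant.
# On play, an affiliation compare node tests the character's affiliation: a true result passes through a short trigger delay, while false fires immediately.
# Either trigger picks a random impact wave to play, whose end finishes the source.
# The source uses projectile-impact attenuation and overrides concurrency to limit simultaneous, rapidly retriggered instances.
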MetaSoundSource /TowHookWeapon/Assets/Audio/MSS_TowHookWeapon_Impact_Generic
inputs [
 Character.Affiliation {type=Enum:FortAudioAffiliation}
 Character.Health {type=Float}
 Character.UniqueID {type=Int32}
 Sounds {type=WaveAsset:Array}
 UE.Source.OnPlay {type=Trigger}
]
interfaces [Fort.Character:1.0, UE.OutputFormat.Mono:1.0, UE.Source.OneShot:1.0, UE.Source:1.1]
outputs [
 "UE.OutputFormat.Mono.Audio:0" {type=Audio}
 UE.Source.OneShot.OnFinished {type=Trigger}
]
pages [Default {
  input_defaults {
   Character.Affiliation=0
   Character.Health=0
   Character.UniqueID=0
   UE.Source.OnPlay=null
   Sounds [
    /TowHookWeapon/Assets/Audio/wav/TowHook_Proj_Impact_Generic_04
    /TowHookWeapon/Assets/Audio/wav/TowHook_Proj_Impact_Generic_03
    /TowHookWeapon/Assets/Audio/wav/TowHook_Proj_Impact_Generic_02
    /TowHookWeapon/Assets/Audio/wav/TowHook_Proj_Impact_Generic_01
   ]
  }
  nodes [
   n1 = &TriggerCompareAffiliation(A=^Character.Affiliation, Compare=^UE.Source.OnPlay)
   TriggerDelay_0 = UE.Trigger Delay(In=@n1.True) {also {defaults {Delay Time=0.2}}}
   TriggerAnyTriggerAny2_0 = "TriggerAny.Trigger Any (2)"(In 1=UE.Reroute(Value=UE.Reroute(Value=@n1.False).Value).Value, In 0=@TriggerDelay_0.Out)
   ArrayRandomGet_0 = Array.Random Get:WaveAsset:Array(In Array=^Sounds, Next=@TriggerAnyTriggerAny2_0.Out)
   WavePlayer_0 = UE.Wave Player:Mono(Play=@ArrayRandomGet_0.On Next, Wave Asset=@ArrayRandomGet_0.Value)
  ]
  links [{dst="@UE.Source.OneShot.OnFinished", src=WavePlayer_0.On Finished}, {dst="@UE.OutputFormat.Mono.Audio:0", src=WavePlayer_0.Out Mono}]
 }]
properties {
 assetImportData=&self:AssetImportData
 attenuationSettings=&/Att_TowHook_Projectile_Impacts
 bOverrideConcurrency=true
 maxDistance=5000
 sampleRate=48000
 soundClassObject=&SFX
 volume=2.2
 concurrencyOverrides {maxCount=2, retriggerTime=0.3}
}
