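# One-shot stereo MetaSound for the tow hook weapon's first-person impact on a player, drawing on body-impact samples.
# On play, a short Trigger Delay fires a random pick from the wave array into a stereo Wave Player, whose finish ends the source.
# It uses projectile-impact 1P attenuation and the impacts sound class, with slightly lowered pitch, boosted volume and overridden concurrency limiting voices and retriggers.
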MetaSoundSource /TowHookWeapon/Assets/Audio/MSS_TowHookWeapon_Impact_Player_1P
inputs [
 Sounds {type=WaveAsset:Array}
 UE.Source.OnPlay {type=Trigger}
]
interfaces [UE.OutputFormat.Stereo:1.0, UE.Source.OneShot:1.0, UE.Source:1.1]
outputs [
 "UE.OutputFormat.Stereo.Audio:0" {type=Audio}
 "UE.OutputFormat.Stereo.Audio:1" {type=Audio}
 UE.Source.OneShot.OnFinished {type=Trigger}
]
pages [Default {
  input_defaults {
   UE.Source.OnPlay=null
   Sounds [
    /TowHookWeapon/Assets/Audio/wav/TowHook_Proj_Impact_Body_1P_01
    /TowHookWeapon/Assets/Audio/wav/TowHook_Proj_Impact_Body_1P_02
    /TowHookWeapon/Assets/Audio/wav/TowHook_Proj_Impact_Body_1P_03
   ]
  }
  nodes [
   TriggerDelay_0 = UE.Trigger Delay(In=^UE.Source.OnPlay) {also {defaults {Delay Time=0.1}}}
   ArrayRandomGet_0 = Array.Random Get:WaveAsset:Array(In Array=^Sounds, Next=@TriggerDelay_0.Out)
   WavePlayer_0 = UE.Wave Player:Stereo(Play=@ArrayRandomGet_0.On Next, Wave Asset=@ArrayRandomGet_0.Value)
  ]
  links [
   {dst="@UE.Source.OneShot.OnFinished", src=WavePlayer_0.On Finished}
   {dst="@UE.OutputFormat.Stereo.Audio:0", src=WavePlayer_0.Out Left}
   {dst="@UE.OutputFormat.Stereo.Audio:1", src=WavePlayer_0.Out Right}
  ]
 }]
properties {
 assetImportData=&self:AssetImportData
 attenuationSettings=&/Att_TowHook_Projectile_Impacts_1P
 bOverrideConcurrency=true
 numChannels=2
 outputFormat=Stereo
 pitch=0.95
 sampleRate=48000
 soundClassObject=&Fort_Impacts
 volume=2
 concurrencyOverrides {maxCount=5, retriggerTime=0.3}
}
