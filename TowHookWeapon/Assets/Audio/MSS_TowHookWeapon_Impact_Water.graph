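# One-shot mono MetaSound for the tow hook weapon's water impact, using bullet-impact attenuation and the Fort_Impacts sound class.
# On play, it picks a random wave from the Sounds array, which by default holds a single borrowed ski water-landing impact.
# A small random pitch shift varies each hit before a Volume input scales the output.
# The source reports finished when the wave player completes.
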MetaSoundSource /TowHookWeapon/Assets/Audio/MSS_TowHookWeapon_Impact_Water
inputs [
 Sounds {type=WaveAsset:Array}
 UE.Source.OnPlay {type=Trigger}
 Volume {type=Float}
]
interfaces [UE.OutputFormat.Mono:1.0, UE.Source.OneShot:1.0, UE.Source:1.1]
outputs [
 "UE.OutputFormat.Mono.Audio:0" {type=Audio}
 UE.Source.OneShot.OnFinished {type=Trigger}
]
pages [Default {
  input_defaults {UE.Source.OnPlay=null, Volume=1.576, Sounds [/FNGameplayCues/Athena/Sounds/Water/Interaction/Ski_Water_Jump_Land_Impact_03]}
  nodes [
   ArrayRandomGet_0 = Array.Random Get:WaveAsset:Array(In Array=^Sounds, Next=^UE.Source.OnPlay)
   n1 = UE.RandomFloat(Next=@ArrayRandomGet_0.On Next) {also {defaults {Max=0.1, Min=-0.1}}}
   WavePlayer_0 = UE.Wave Player:Mono(Wave Asset=@ArrayRandomGet_0.Value, Play=@n1.On Next, Pitch Shift=@n1.Value)
   Multiply_0 = UE.Multiply:Audio by Float(AdditionalOperands=^Volume, PrimaryOperand=@WavePlayer_0.Out Mono)
  ]
  links [{dst="@UE.OutputFormat.Mono.Audio:0", src=Multiply_0.Out}, {dst="@UE.Source.OneShot.OnFinished", src=WavePlayer_0.On Finished}]
 }]
properties {
 assetImportData=&self:AssetImportData
 attenuationSettings=/Game/Sounds/Attenuation/Fort_Bullet_Impact_Att
 maxDistance=5300
 sampleRate=48000
 soundClassObject=&Fort_Impacts
}
