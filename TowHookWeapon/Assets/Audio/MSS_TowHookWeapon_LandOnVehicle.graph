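# One-shot mono MetaSound for the tow hook weapon's land-on-vehicle sound, pairing a jump twirl with a delayed landing.
# On play it starts a random twirl wave, then after a short delay a random landing wave, each with randomized pitch shift.
# Both players mix to mono with the twirl boosted, and a trigger accumulator ends the sound once both finish.
# It reuses Fortnite's third-person player landing attenuation and overrides concurrency to limit simultaneous instances.
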MetaSoundSource /TowHookWeapon/Assets/Audio/MSS_TowHookWeapon_LandOnVehicle
inputs [
 Lands {type=WaveAsset:Array}
 Sounds {type=WaveAsset:Array}
 UE.Source.OnPlay {type=Trigger}
]
interfaces [UE.OutputFormat.Mono:1.0, UE.Source.OneShot:1.0, UE.Source:1.1]
outputs [
 "UE.OutputFormat.Mono.Audio:0" {type=Audio}
 UE.Source.OneShot.OnFinished {type=Trigger}
]
pages [Default {
  input_defaults {
   UE.Source.OnPlay=null
   Lands [/TowHookWeapon/Assets/Audio/wav/TowHook_Player_Jump_Twirl_Land_01, /TowHookWeapon/Assets/Audio/wav/TowHook_Player_Jump_Twirl_Land_02]
   Sounds [
    /TowHookWeapon/Assets/Audio/wav/TowHook_Player_Jump_Twirl_01
    /TowHookWeapon/Assets/Audio/wav/TowHook_Player_Jump_Twirl_02
    /TowHookWeapon/Assets/Audio/wav/TowHook_Player_Jump_Twirl_03
   ]
  }
  nodes [
   ArrayRandomGet_0 = Array.Random Get:WaveAsset:Array(In Array=^Sounds, Next=^UE.Source.OnPlay)
   n1 = UE.RandomFloat(Next=@ArrayRandomGet_0.On Next) {also {defaults {Max=0.5, Min=-0.5}}}
   WavePlayer_0 = UE.Wave Player:Mono(Wave Asset=@ArrayRandomGet_0.Value, Play=@n1.On Next, Pitch Shift=@n1.Value)
   TriggerDelay_0 = UE.Trigger Delay(In=^UE.Source.OnPlay) {also {defaults {Delay Time=0.6}}}
   ArrayRandomGet_1 = Array.Random Get:WaveAsset:Array(In Array=^Lands, Next=@TriggerDelay_0.Out)
   n2 = UE.RandomFloat(Next=@ArrayRandomGet_1.On Next) {also {defaults {Max=0.5, Min=-0.5}}}
   WavePlayer_1 = UE.Wave Player:Mono(Wave Asset=@ArrayRandomGet_1.Value, Play=@n2.On Next, Pitch Shift=@n2.Value)
   AudioMixerAudioMixerMono2_0 = "AudioMixer.Audio Mixer (Mono, 2)"(In 0=@WavePlayer_0.Out Mono, In 1=@WavePlayer_1.Out Mono) {also {defaults {Gain 0=1.5}}}
   TriggerAccumulatorTriggerAccumulate2_0 = "TriggerAccumulator.Trigger Accumulate (2)"(In 0=@WavePlayer_0.On Finished, In 1=@WavePlayer_1.On Finished)
  ]
  links [
   {dst="@UE.OutputFormat.Mono.Audio:0", src=AudioMixerAudioMixerMono2_0.Out}
   {dst="@UE.Source.OneShot.OnFinished", src=TriggerAccumulatorTriggerAccumulate2_0.Out}
  ]
 }]
properties {
 assetImportData=&self:AssetImportData
 attenuationSettings=/Game/Sounds/Attenuation/Fort_Player_Land_3P_Athena_Att
 bOverrideConcurrency=true
 maxDistance=3584
 sampleRate=48000
 soundClassObject=&SFX
 volume=2
 concurrencyOverrides {maxCount=2}
}
