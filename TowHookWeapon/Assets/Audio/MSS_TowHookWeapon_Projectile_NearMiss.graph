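# One-shot mono MetaSound for the tow hook weapon's projectile near-miss whoosh, with its own attenuation and concurrency limit.
# A fixed threshold on attenuation distance, or the Fire event, feeds a trigger gate that closes shortly afterward with no reopen.
# The gated trigger picks a random by-pass wave and plays it, finishing the source when done.
# The random float only relays the play trigger, and the Threshold, Distance and Stop inputs are unwired.
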MetaSoundSource /TowHookWeapon/Assets/Audio/MSS_TowHookWeapon_Projectile_NearMiss
inputs [
 Distance {type=Float}
 Events.Fire {type=Trigger}
 Sounds {type=WaveAsset:Array}
 Stop {type=Trigger}
 Threshold {type=Float}
 UE.Attenuation.Distance {type=Float}
 UE.Source.OnPlay {type=Trigger}
]
interfaces [UE.Attenuation:1.0, UE.OutputFormat.Mono:1.0, UE.Source.OneShot:1.0, UE.Source:1.1]
outputs [
 "UE.OutputFormat.Mono.Audio:0" {type=Audio}
 UE.Source.OneShot.OnFinished {type=Trigger}
]
pages [Default {
  input_defaults {
   Distance=0
   Events.Fire=null
   Stop=null
   Threshold=1000
   UE.Attenuation.Distance=648.0006
   UE.Source.OnPlay=null
   Sounds [
    /TowHookWeapon/Assets/Audio/wav/TowHook_Proj_By_03
    /TowHookWeapon/Assets/Audio/wav/TowHook_Proj_By_02
    /TowHookWeapon/Assets/Audio/wav/TowHook_Proj_By_01
   ]
  }
  nodes [
   n1 = InitVariable.Trigger()
   TriggerAnyTriggerAny2_0 = "TriggerAny.Trigger Any (2)"(In 1=^Events.Fire, In 0=UE.TriggerOnThreshold:Float(In=^UE.Attenuation.Distance) {also {defaults {Threshold=1000, Type=1}}}.Out)
   TriggerDelay_0 = UE.Trigger Delay(In=@TriggerAnyTriggerAny2_0.Out) {also {defaults {Delay Time=0.05}}}
   TriggerControl_0 = UE.Trigger Control(Trigger In=@TriggerAnyTriggerAny2_0.Out, Close=@TriggerDelay_0.Out)
   ArrayRandomGet_0 = Array.Random Get:WaveAsset:Array(
    In Array=^Sounds
    Next=VariableAccessor.Trigger(Variable=VariableMutator.Trigger(Variable=@n1.Variable, Value=@TriggerControl_0.Trigger Out).Variable).Value
   )
   WavePlayer_0 = UE.Wave Player:Mono(Wave Asset=@ArrayRandomGet_0.Value, Play=UE.RandomFloat(Next=@ArrayRandomGet_0.On Next) {also {defaults {Max=0.5, Min=-0.5}}}.On Next)
  ]
  links [{dst="@UE.Source.OneShot.OnFinished", src=WavePlayer_0.On Finished}, {dst="@UE.OutputFormat.Mono.Audio:0", src=WavePlayer_0.Out Mono}]
 }]
properties {
 assetImportData=&self:AssetImportData
 attenuationSettings=&/Att_TowHook_Projectile_NearMiss
 bOverrideConcurrency=true
 maxDistance=100000
 sampleRate=48000
 soundClassObject=&Fort_WeaponRanged_Projectile_Effects
 volume=0.6
 concurrencyOverrides {maxCount=3, resolutionRule=StopOldest}
}
