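# Translucent unlit smoke ring material for Niagara mesh particles, tinted by particle color through an emissive adjustment function.
# Opacity multiplies a puffy trail texture by particle alpha and an erosion mask, made by subtracting panning cloud noise from the first dynamic parameter.
# The second dynamic parameter drives SplineThicken width at base and tip.
# Depth fade softens opacity, except on low and medium quality, which skip it.
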Material /TowHookWeapon/Assets/FX/Materials/M_SmokeRing
graph {
 properties {
  bUsedWithNiagaraMeshParticles=true
  blendMode=BLEND_Translucent
  pixelDepthOffsetMode=PDOM_Legacy
  shadingModel=MSM_Unlit
  usedShadingModels=Unlit
 }
 nodes [
  Constant() {r=1}
  n1 = ParticleColor()
  MaterialFunctionCall(Emissive=@n1.RGB) {materialFunction=&MF_FX_EmissiveAdjustments}
  n2 = DynamicParameter() {defaultValue=<rgba 1 20 1 1>}
  MaterialFunctionCall(WidthBase=@n2.Param2, WidthTip=@n2.Param2, Expand U or V UV Channel=StaticBool()) {materialFunction=&SplineThicken}
  n3 = TextureCoordinate() {uTiling=2}
  n4 = Subtract(A=@n2.Param1, B=TextureSample(UVs=Panner(Coordinate=@n3) {speedY=-1}) {texture=/Game/Effects/Fort_Effects/Textures/Smoke/T_Clouds_02_M}.G)
  n5 = Multiply(A=Multiply(A=TextureSample() {texture=/Game/Effects/Fort_Effects/Textures/Smoke/T_Puffy_Smoke_Trail_01}.G, B=Saturate(None=@n4)), B=@n1.A)
  QualitySwitch(Default=DepthFade(Opacity=@n5) {fadeDistanceDefault=250}, Low=@n5, Medium=@n5)
  TextureCoordinate() {coordinateIndex=1}
  TextureSample() {texture=/Game/Effects/Fort_Effects/Textures/Smoke/T_Smoke_Bomb_Disk2}
 ]
}
