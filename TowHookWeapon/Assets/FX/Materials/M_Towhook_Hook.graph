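# Default Lit surface material for the tow hook weapon's hook, with usage flags for Niagara mesh particles, ribbons and sprites.
# It samples the tow hook's diffuse and normal textures for color and surface detail.
# The blue channel of the mask texture blends between minimum and maximum roughness parameters, so the roughness range is tunable per instance.
# Automatic usage setting in the editor is disabled.
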Material /TowHookWeapon/Assets/FX/Materials/M_Towhook_Hook
graph {
 properties {
  bAutomaticallySetUsageInEditor=false
  bUsedWithNiagaraMeshParticles=true
  bUsedWithNiagaraRibbons=true
  bUsedWithNiagaraSprites=true
  pixelDepthOffsetMode=PDOM_Legacy
  usedShadingModels=Default Lit
 }
 nodes [
  LinearInterpolate(
   A=ScalarParameter() {group=Roughness, parameterName=SpecRoughnessMin}
   B=ScalarParameter() {defaultValue=1, group=Roughness, parameterName=SpecRoughnessMax}
   Alpha=TextureSample() {samplerType=SAMPLERTYPE_Masks, texture=/TowHookWeapon/Assets/Textures/T_TowHook_S}.B
  )
  TextureSample() {samplerType=SAMPLERTYPE_Normal, texture=/TowHookWeapon/Assets/Textures/T_TowHook_N}
  TextureSample() {texture=/TowHookWeapon/Assets/Textures/T_TowHook_D}
 ]
}
