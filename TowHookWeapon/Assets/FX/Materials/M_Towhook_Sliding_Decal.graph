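# Translucent deferred decal material for the tow hook's sliding marks.
# Tiled UVs scrolled over time by a vector parameter sample the sliding texture, whose red channel blends two dark colors.
# The scrolling green channel, masked by the same texture's static, vertically offset blue channel and the decal color alpha, forms an apparent opacity.
# Both axes share the U tile and speed values, and a separate radial-gradient branch has no visible consumer.
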Material /TowHookWeapon/Assets/FX/Materials/M_Towhook_Sliding_Decal
graph {
 properties {blendMode=BLEND_Translucent, materialDomain=MD_DeferredDecal, pixelDepthOffsetMode=PDOM_Legacy, usedShadingModels=Default Lit}
 nodes [
  n1 = TextureCoordinate()
  n2 = VectorParameter() {group=Noise2, parameterName=Tile UV, channelNames {a=Speed V, b=Speed U, g=Tile V, r=Tile U}, defaultValue=<rgba 1 1 0 -0.5>}
  n3 = MaterialFunctionCall(X=@n2.Tile U, Y=@n2.Tile U) {materialFunction=&MakeFloat2}
  n4 = MaterialFunctionCall(X=@n2.Speed U, Y=@n2.Speed U) {materialFunction=&MakeFloat2}
  n5 = Time()
  n6 = TextureSample(UVs=Add(A=Multiply(A=@n1, B=@n3.Result), B=Multiply(A=@n4.Result, B=@n5))) {texture=&/T_Towhook_Sliding_Decal}
  LinearInterpolate(
   A=Constant3Vector() {constant=<rgba 0.025 0.025 0.0229 1>}
   B=Constant3Vector() {constant=<rgba 0.075 0.0682 0.0628 1>}
   Alpha=Multiply(A=@n6.R) {constB=15}
  )
  n7 = Add(A=Multiply(A=TextureCoordinate(), B=Constant2Vector() {g=0.8, r=1}), B=Constant2Vector() {g=0.2})
  Multiply(
   A=Multiply(A=Saturate(None=@n6.G), B=TextureSample(UVs=@n7) {texture=&T_Towhook_Sliding_Decal}.B)
   B=DecalColor().A
  )
  n8 = ComponentMask(None=TextureCoordinate() {coordinateIndex=1}) {g=true}
  Multiply(A=TextureSample() {texture=/Game/Athena/HUD/Materials/T_UI_RadialGradient}.RGB, B=Multiply(A=Add(A=@n8) {constB=-0.45}) {constB=10})
 ]
}
